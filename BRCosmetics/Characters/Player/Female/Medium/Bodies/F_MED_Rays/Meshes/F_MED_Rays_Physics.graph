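# Physics asset for the F_MED_Rays female medium player body cosmetic, covering its dangling frills, front flaps and suspenders.
# Capsules on pelvis, spine and thighs form body shapes, with small helper bodies at sub-thigh and chain roots.
# Side and rear frills hang from sub-thighs, center back frill from pelvis, suspenders from spine_02.
# Flaps and suspenders are two-segment, angle-limited chains with tight twist, while spine and suspender-root joints rotate freely.
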PhysicsAsset /BRCosmetics/Characters/Player/Female/Medium/Bodies/F_MED_Rays/Meshes/F_MED_Rays_Physics
bodies {
 Flap_Front_Driven_Bone [{shapeType=Capsule, length=0.505, radius=0.505, rotation=<rot 0 -90 0>}]
 dyn_flap_frnt_01_l [{shapeType=Capsule, length=3.5, radius=2.5, shapeName=dyn_flap_frnt_01_l_capsule, center=<2.943 0 0>, rotation=<rot 90 0 0>}]
 dyn_flap_frnt_01_r [{shapeType=Capsule, length=2.25, radius=2.75, shapeName=dyn_flap_frnt_01_r_capsule, center=<3.9734 0 0>, rotation=<rot 90 0 0>}]
 dyn_flap_frnt_base_l [{shapeType=Sphere, radius=0.5, shapeName=dyn_flap_frnt_base_l_sphere}]
 dyn_flap_frnt_base_r [{shapeType=Sphere, radius=0.5, shapeName=dyn_flap_frnt_base_r_sphere}]
 dyn_flap_frnt_parent_l [{shapeType=Capsule, length=3.5, radius=2, shapeName=dyn_flap_frnt_parent_l_capsule, center=<1.9932 0 0>, rotation=<rot 90 0 0>}]
 dyn_flap_frnt_parent_r [{shapeType=Capsule, length=5, radius=1.75, shapeName=dyn_flap_frnt_parent_r_capsule, center=<1.602 0 0>, rotation=<rot 0 0 90>}]
 dyn_frill_bck [{shapeType=Capsule, length=5.2413, radius=2.8791, center=<4.3526 0 -0.4373>, rotation=<rot 84.2633 0 0>}]
 dyn_frill_bck_l [{shapeType=Capsule, length=5.0675, radius=2, center=<3.5667 0 0>, rotation=<rot 90 0 0>}]
 dyn_frill_bck_r [{shapeType=Capsule, length=4.8399, radius=2, center=<3.4793 0.0002 0>, rotation=<rot 90 0 0>}]
 dyn_frill_side_l [{shapeType=Capsule, length=4.5276, radius=2, center=<3.484 0.0001 0>, rotation=<rot 90 0 0>}]
 dyn_frill_side_r [{shapeType=Capsule, length=4.4908, radius=2, center=<3.5835 0 0>, rotation=<rot 90 0 0>}]
 dyn_suspender_01_l [{shapeType=Capsule, length=11.3326, radius=3.4199, shapeName=dyn_suspender_01_l_capsule, center=<3 -0.26 -2.25>, rotation=<rot 0 -4.3509 90>}]
 dyn_suspender_01_r [{
   shapeType=Capsule
   length=8.6999
   radius=3.4438
   shapeName=dyn_suspender_01_r_capsule
   center=<3.29 -0.55 -1.8>
   rotation=<rot 0 -178.2101 90>
  }]
 dyn_suspender_02_l [{shapeType=Capsule, length=7.6286, radius=3.5465, shapeName=dyn_suspender_02_l_capsule, center=<2.6 -0.15 -0.5>, rotation=<rot 0 -3.317 90>}]
 dyn_suspender_02_r [{shapeType=Capsule, length=11.0526, radius=2.8161, shapeName=dyn_suspender_02_r_capsule, center=<2 0.04 0>, rotation=<rot 0 1.3162 90>}]
 dyn_suspender_parent_l [{shapeType=Sphere, radius=0.5, shapeName=dyn_suspender_parent_l_sphere}]
 dyn_suspender_parent_r [{shapeType=Sphere, radius=0.5, shapeName=dyn_suspender_parent_r_sphere}]
 pelvis [{shapeType=Capsule, length=11.75, radius=8.5, center=<-2.606 -1.5942 0>}]
 spine_01 [{shapeType=Capsule, length=2.3139, radius=9.3918, shapeName=spine_01_capsule, rotation=<rot 0 -90 0>}]
 spine_02 [{shapeType=Capsule, length=11.5472, radius=7.9967, shapeName=spine_02_capsule, rotation=<rot 90 0 0>}]
 sub_thigh_l [{shapeType=Capsule, length=0.505, radius=0.505, rotation=<rot 0 -90 0>}]
 sub_thigh_r [{shapeType=Capsule, length=0.505, radius=0.505, rotation=<rot 0 -90 0>}]
 thigh_l [{shapeType=Capsule, length=6.6409, radius=7.9362, center=<-15.1678 0 0>, rotation=<rot 0 90 90>}]
 thigh_r [{shapeType=Capsule, length=7.028, radius=7.4212, center=<14.9946 0 0>, rotation=<rot 0 90 90>}]
}
constraints [
 {bone1Name=sub_thigh_l, bone2Name=thigh_l, swing1LimitDegrees=45, swing2LimitDegrees=45, twistLimitDegrees=45}
 {bone1Name=sub_thigh_r, bone2Name=thigh_r, swing1LimitDegrees=45, swing2LimitDegrees=45, twistLimitDegrees=45}
 {bone1Name=dyn_frill_bck_l, bone2Name=sub_thigh_l, swing1LimitDegrees=50, swing2LimitDegrees=50, twistLimitDegrees=45}
 {bone1Name=dyn_frill_side_l, bone2Name=sub_thigh_l, swing1LimitDegrees=45, swing2LimitDegrees=45, twistLimitDegrees=15}
 {bone1Name=dyn_frill_bck_r, bone2Name=sub_thigh_r, swing1LimitDegrees=50, swing2LimitDegrees=50, twistLimitDegrees=45}
 {bone1Name=dyn_frill_side_r, bone2Name=sub_thigh_r, swing1LimitDegrees=45, swing2LimitDegrees=45, twistLimitDegrees=15}
 {bone1Name=dyn_frill_bck, bone2Name=pelvis, swing1LimitDegrees=45, swing2LimitDegrees=45, twistLimitDegrees=45}
 {bone1Name=dyn_flap_frnt_parent_l, bone2Name=dyn_flap_frnt_base_l, swing1LimitDegrees=45, swing2LimitDegrees=45, twistLimitDegrees=3}
 {bone1Name=dyn_flap_frnt_01_l, bone2Name=dyn_flap_frnt_parent_l, swing1LimitDegrees=20, swing2LimitDegrees=20, twistLimitDegrees=3}
 {bone1Name=dyn_flap_frnt_parent_r, bone2Name=dyn_flap_frnt_base_r, swing1LimitDegrees=45, swing2LimitDegrees=45, twistLimitDegrees=3}
 {bone1Name=dyn_flap_frnt_01_r, bone2Name=dyn_flap_frnt_parent_r, swing1LimitDegrees=20, swing2LimitDegrees=20, twistLimitDegrees=3}
 {
  bone1Name=spine_01
  bone2Name=pelvis
  swing1LimitDegrees=45
  swing1Motion=Free
  swing2LimitDegrees=45
  swing2Motion=Free
  twistLimitDegrees=45
  twistMotion=Free
 }
 {
  bone1Name=spine_02
  bone2Name=spine_01
  swing1LimitDegrees=45
  swing1Motion=Free
  swing2LimitDegrees=45
  swing2Motion=Free
  twistLimitDegrees=45
  twistMotion=Free
 }
 {bone1Name=dyn_suspender_01_l, bone2Name=dyn_suspender_parent_l, swing1LimitDegrees=5, swing2LimitDegrees=25, twistLimitDegrees=3}
 {bone1Name=dyn_suspender_02_l, bone2Name=dyn_suspender_01_l, swing1LimitDegrees=15, swing2LimitDegrees=15, twistLimitDegrees=3}
 {bone1Name=dyn_suspender_01_r, bone2Name=dyn_suspender_parent_r, swing1LimitDegrees=5, swing2LimitDegrees=25, twistLimitDegrees=3}
 {bone1Name=dyn_suspender_02_r, bone2Name=dyn_suspender_01_r, swing1LimitDegrees=15, swing2LimitDegrees=15, twistLimitDegrees=3}
 {
  bone1Name=dyn_suspender_parent_l
  bone2Name=spine_02
  swing1LimitDegrees=45
  swing1Motion=Free
  swing2LimitDegrees=45
  swing2Motion=Free
  twistLimitDegrees=45
  twistMotion=Free
 }
 {
  bone1Name=dyn_suspender_parent_r
  bone2Name=spine_02
  swing1LimitDegrees=45
  swing1Motion=Free
  swing2LimitDegrees=45
  swing2Motion=Free
  twistLimitDegrees=45
  twistMotion=Free
 }
]
