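# Physics asset for the front loincloth flaps on the medium female player Rays body.
# Left and right flaps each form a base, parent and tip chain, using small spheres at the bases and capsules below.
# Constraints allow wider swing at the upper joint, tighter at the tip, with very little twist.
# Unconstrained pelvis and thigh capsules supply body volumes, alongside a tiny capsule on the front flap driven bone.
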PhysicsAsset /Game/Characters/Player/Female/Medium/Bodies/F_MED_Rays/Meshes/F_MED_Rays_LoinCloth_Physics
bodies {
 Flap_Front_Driven_Bone [{shapeType=Capsule, length=0.505, radius=0.505, rotation=<rot 0 -90 0>}]
 dyn_flap_frnt_01_l [{shapeType=Capsule, length=3.5, radius=2.5, shapeName=dyn_flap_frnt_01_l_capsule, center=<2.943 0 0>, rotation=<rot 90 0 0>}]
 dyn_flap_frnt_01_r [{shapeType=Capsule, length=2.25, radius=2.75, shapeName=dyn_flap_frnt_01_r_capsule, center=<3.9734 0 0>, rotation=<rot 90 0 0>}]
 dyn_flap_frnt_base_l [{shapeType=Sphere, radius=0.5, shapeName=dyn_flap_frnt_base_l_sphere}]
 dyn_flap_frnt_base_r [{shapeType=Sphere, radius=0.5, shapeName=dyn_flap_frnt_base_r_sphere}]
 dyn_flap_frnt_parent_l [{shapeType=Capsule, length=3.5, radius=2, shapeName=dyn_flap_frnt_parent_l_capsule, center=<1.9932 0 0>, rotation=<rot 90 0 0>}]
 dyn_flap_frnt_parent_r [{shapeType=Capsule, length=5, radius=1.75, shapeName=dyn_flap_frnt_parent_r_capsule, center=<1.602 0 0>, rotation=<rot 0 0 90>}]
 pelvis [{shapeType=Capsule, length=23.5, radius=8.25, shapeName=pelvis_capsule, center=<-8.093 -2.4823 0>}]
 thigh_l [{shapeType=Capsule, length=19.1409, radius=7.9362, center=<-15.1597 -1.0903 -0.2196>, rotation=<rot 0 87.1871 95.625>}]
 thigh_r [{shapeType=Capsule, length=19.278, radius=7.6712, center=<15.0004 0.3672 -0.2247>, rotation=<rot 0.2755 87.2008 95.6183>}]
}
constraints [
 {bone1Name=dyn_flap_frnt_parent_r, bone2Name=dyn_flap_frnt_base_r, swing1LimitDegrees=45, swing2LimitDegrees=45, twistLimitDegrees=3}
 {bone1Name=dyn_flap_frnt_01_r, bone2Name=dyn_flap_frnt_parent_r, swing1LimitDegrees=20, swing2LimitDegrees=20, twistLimitDegrees=3}
 {bone1Name=dyn_flap_frnt_parent_l, bone2Name=dyn_flap_frnt_base_l, swing1LimitDegrees=45, swing2LimitDegrees=45, twistLimitDegrees=3}
 {bone1Name=dyn_flap_frnt_01_l, bone2Name=dyn_flap_frnt_parent_l, swing1LimitDegrees=20, swing2LimitDegrees=20, twistLimitDegrees=3}
]
