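# Physics asset for the skirt of the F_MED_Rays female player body, providing capsule bodies and constraints for its frill bones.
# Larger pelvis and thigh capsules approximate the body, alongside five small frill capsules.
# Side and back-corner frills constrain to tiny sub_thigh bodies linked to the thighs, while the center-back frill constrains to the pelvis.
# Swing and twist limits are specified per joint; the tiny Flap_Front_Driven_Bone body has no constraint.
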PhysicsAsset /Game/Characters/Player/Female/Medium/Bodies/F_MED_Rays/Meshes/F_MED_Rays_Skirt_Physics
bodies {
 Flap_Front_Driven_Bone [{shapeType=Capsule, length=0.505, radius=0.505, rotation=<rot 0 -90 0>}]
 dyn_frill_bck [{shapeType=Capsule, length=5.2413, radius=2.8791, center=<4.3526 0 -0.4373>, rotation=<rot 84.2633 0 0>}]
 dyn_frill_bck_l [{shapeType=Capsule, length=5.0675, radius=2, center=<3.5667 0 0>, rotation=<rot 90 0 0>}]
 dyn_frill_bck_r [{shapeType=Capsule, length=4.8399, radius=2, center=<3.4793 0.0002 0>, rotation=<rot 90 0 0>}]
 dyn_frill_side_l [{shapeType=Capsule, length=4.5276, radius=2, center=<3.484 0.0001 0>, rotation=<rot 90 0 0>}]
 dyn_frill_side_r [{shapeType=Capsule, length=4.4908, radius=2, center=<3.5835 0 0>, rotation=<rot 90 0 0>}]
 pelvis [{shapeType=Capsule, length=11.75, radius=8.5, center=<-2.606 -1.5942 0>}]
 sub_thigh_l [{shapeType=Capsule, length=0.505, radius=0.505, rotation=<rot 0 -90 0>}]
 sub_thigh_r [{shapeType=Capsule, length=0.505, radius=0.505, rotation=<rot 0 -90 0>}]
 thigh_l [{shapeType=Capsule, length=6.6409, radius=7.9362, center=<-15.1678 0 0>, rotation=<rot 0 90 90>}]
 thigh_r [{shapeType=Capsule, length=7.028, radius=7.4212, center=<14.9946 0 0>, rotation=<rot 0 90 90>}]
}
constraints [
 {bone1Name=sub_thigh_l, bone2Name=thigh_l, swing1LimitDegrees=45, swing2LimitDegrees=45, twistLimitDegrees=45}
 {bone1Name=sub_thigh_r, bone2Name=thigh_r, swing1LimitDegrees=45, swing2LimitDegrees=45, twistLimitDegrees=45}
 {bone1Name=dyn_frill_bck_l, bone2Name=sub_thigh_l, swing1LimitDegrees=50, swing2LimitDegrees=50, twistLimitDegrees=45}
 {bone1Name=dyn_frill_side_l, bone2Name=sub_thigh_l, swing1LimitDegrees=45, swing2LimitDegrees=45, twistLimitDegrees=15}
 {bone1Name=dyn_frill_bck_r, bone2Name=sub_thigh_r, swing1LimitDegrees=50, swing2LimitDegrees=50, twistLimitDegrees=45}
 {bone1Name=dyn_frill_side_r, bone2Name=sub_thigh_r, swing1LimitDegrees=45, swing2LimitDegrees=45, twistLimitDegrees=15}
 {bone1Name=dyn_frill_bck, bone2Name=pelvis, swing1LimitDegrees=45, swing2LimitDegrees=45, twistLimitDegrees=45}
]
