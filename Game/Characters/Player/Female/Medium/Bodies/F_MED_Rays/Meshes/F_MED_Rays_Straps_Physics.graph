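# Physics asset for the strap mesh of the F_MED_Rays female player body, covering its left and right dynamic suspender bones.
# Each side is a two-capsule chain hanging from a small parent sphere, with tight twist and narrow swing limits.
# Both parent spheres attach to spine_01 with free angular motion, as do the pelvis and spine_02 bodies.
# Pelvis and spine shapes give the straps a rough torso volume.
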PhysicsAsset /Game/Characters/Player/Female/Medium/Bodies/F_MED_Rays/Meshes/F_MED_Rays_Straps_Physics
bodies {
 dyn_suspender_01_l [{shapeType=Capsule, length=11.3326, radius=3.4199, shapeName=dyn_suspender_01_l_capsule, center=<3 -0.26 -2.25>, rotation=<rot 0 -4.3509 90>}]
 dyn_suspender_01_r [{
   shapeType=Capsule
   length=8.6999
   radius=3.4438
   shapeName=dyn_suspender_01_r_capsule
   center=<3.29 -0.55 -1.8>
   rotation=<rot 0 -178.2101 90>
  }]
 dyn_suspender_02_l [{shapeType=Capsule, length=7.6286, radius=3.5465, shapeName=dyn_suspender_02_l_capsule, center=<2.6 -0.15 -0.5>, rotation=<rot 0 -3.317 90>}]
 dyn_suspender_02_r [{shapeType=Capsule, length=11.0526, radius=2.8161, shapeName=dyn_suspender_02_r_capsule, center=<2 0.04 0>, rotation=<rot 0 1.3162 90>}]
 dyn_suspender_parent_l [{shapeType=Sphere, radius=0.5, shapeName=dyn_suspender_parent_l_sphere}]
 dyn_suspender_parent_r [{shapeType=Sphere, radius=0.5, shapeName=dyn_suspender_parent_r_sphere}]
 pelvis [{shapeType=Sphere, radius=8.3017, shapeName=pelvis_sphere, center=<1.3079 -4.0568 11.0722>}]
 spine_01 [{shapeType=Capsule, length=2.3139, radius=9.3918, shapeName=spine_01_capsule, rotation=<rot 0 -90 0>}]
 spine_02 [{shapeType=Capsule, length=11.5472, radius=7.9967, shapeName=spine_02_capsule, rotation=<rot 90 0 0>}]
}
constraints [
 {bone1Name=dyn_suspender_01_r, bone2Name=dyn_suspender_parent_r, swing1LimitDegrees=5, swing2LimitDegrees=25, twistLimitDegrees=3}
 {bone1Name=dyn_suspender_02_r, bone2Name=dyn_suspender_01_r, swing1LimitDegrees=15, swing2LimitDegrees=15, twistLimitDegrees=3}
 {bone1Name=dyn_suspender_02_l, bone2Name=dyn_suspender_01_l, swing1LimitDegrees=15, swing2LimitDegrees=15, twistLimitDegrees=3}
 {bone1Name=dyn_suspender_01_l, bone2Name=dyn_suspender_parent_l, swing1LimitDegrees=5, swing2LimitDegrees=25, twistLimitDegrees=3}
 {
  bone1Name=spine_02
  bone2Name=spine_01
  swing1LimitDegrees=45
  swing1Motion=Free
  swing2LimitDegrees=45
  swing2Motion=Free
  twistLimitDegrees=45
  twistMotion=Free
 }
 {
  bone1Name=dyn_suspender_parent_r
  bone2Name=spine_01
  swing1LimitDegrees=45
  swing1Motion=Free
  swing2LimitDegrees=45
  swing2Motion=Free
  twistLimitDegrees=45
  twistMotion=Free
 }
 {
  bone1Name=dyn_suspender_parent_l
  bone2Name=spine_01
  swing1LimitDegrees=45
  swing1Motion=Free
  swing2LimitDegrees=45
  swing2Motion=Free
  twistLimitDegrees=45
  twistMotion=Free
 }
 {
  bone1Name=spine_01
  bone2Name=pelvis
  swing1LimitDegrees=45
  swing1Motion=Free
  swing2LimitDegrees=45
  swing2Motion=Free
  twistLimitDegrees=45
  twistMotion=Free
 }
]
