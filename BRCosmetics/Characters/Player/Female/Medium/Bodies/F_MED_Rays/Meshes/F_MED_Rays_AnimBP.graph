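# Post-process-style cosmetic AnimBlueprint for the F_MED_Rays female player body, animating buckles, flaps, frills and suspenders over the attached parent's copied pose.
# Bone constraints feed a RigidBody simulation, a Control Rig driven by crouch, pistol and tactical-sprint values, then Look-At and additive bone fixes.
# Sliding fades out the RigidBody and enables suspender AnimDynamics through inverse alphas.
# The separate straps, skirt and loincloth RigidBody chain is unconnected to the output.
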AnimBlueprint /BRCosmetics/Characters/Player/Female/Medium/Bodies/F_MED_Rays/Meshes/F_MED_Rays_AnimBP
target_skeleton=&/F_MED_Rays_Skeleton
graphs [
 AnimGraph {nodes [
   n1 = RigidBody() {
    bClampLinearTranslationLimitToRefPose=true
    bOverrideWorldGravity=true
    overridePhysicsAsset=/Game/Characters/Player/Female/Medium/Bodies/F_MED_Rays/Meshes/F_MED_Rays_Straps_Physics
    simulationSpace=BaseBoneSpace
    alphaBoolBlend {blendOption=Linear}
    baseBoneRef {boneName=pelvis}
    componentAppliedLinearAccClamp=<3000 3000 3000>
    componentLinearAccScale=<0.4 0.3 0.3>
    componentLinearVelScale=<0.4 0.3 0.3>
    overrideWorldGravity {z=-980}
    simSpaceSettings {
     maxAngularAcceleration=500
     maxAngularVelocity=3
     maxLinearAcceleration=400
     maxLinearVelocity=250
     velocityScaleZ=0
     worldAlpha=1
     externalLinearDragV=<0.1 0.1>
    }
   }
   n2 = RigidBody(ComponentPose=@n1, Alpha=Get SkirtAlpha() -> (SkirtAlpha:float)) {
    bClampLinearTranslationLimitToRefPose=true
    bOverrideWorldGravity=true
    overridePhysicsAsset=/Game/Characters/Player/Female/Medium/Bodies/F_MED_Rays/Meshes/F_MED_Rays_Skirt_Physics
    simulationSpace=BaseBoneSpace
    alphaBoolBlend {blendOption=Linear}
    alphaScaleBiasClamp {bInterpResult=true, interpSpeedDecreasing=5, interpSpeedIncreasing=5}
    baseBoneRef {boneName=pelvis}
    componentAppliedLinearAccClamp=<3000 3000 3000>
    componentLinearAccScale=<0.35 0.3 0.4>
    componentLinearVelScale=<0.35 0.3 0.4>
    overrideWorldGravity {z=-980}
    simSpaceSettings {
     maxAngularAcceleration=500
     maxAngularVelocity=3
     maxLinearAcceleration=400
     maxLinearVelocity=250
     velocityScaleZ=0
     worldAlpha=1
     externalLinearDragV=<0.1 0.1>
    }
   }
   RigidBody(ComponentPose=@n2) {
    bClampLinearTranslationLimitToRefPose=true
    bOverrideWorldGravity=true
    overridePhysicsAsset=/Game/Characters/Player/Female/Medium/Bodies/F_MED_Rays/Meshes/F_MED_Rays_LoinCloth_Physics
    simulationSpace=BaseBoneSpace
    alphaBoolBlend {blendOption=Linear}
    baseBoneRef {boneName=pelvis}
    componentAppliedLinearAccClamp=<3000 3000 3000>
    componentLinearAccScale=<0.35 0.3 0.4>
    componentLinearVelScale=<0.35 0.3 0.4>
    overrideWorldGravity {z=-980}
    simSpaceSettings {
     maxAngularAcceleration=500
     maxAngularVelocity=3
     maxLinearAcceleration=400
     maxLinearVelocity=250
     velocityScaleZ=0
     worldAlpha=1
     externalLinearDragV=<0.1 0.1>
    }
   }
   n3 = "Constraint - const_buckle_front (PP)"!(
    ComponentPose=Local To Component(LocalPose=Copy Pose From Mesh() {bUseAttachedParent=true})
    Alpha=1
    ConstraintWeights_0=0.8282
    ConstraintWeights_1:float=0.1718
   ) -> (Pose) {
    alphaBoolBlend {blendOption=Linear}
    boneToModify {boneName=const_buckle_front}
    constraintSetup [{targetBone {boneName=spine_04}}, {targetBone {boneName=spine_05}}]
    constraintWeights [1,1]
   }
   n4 = "Constraint - const_buckle_back_r (PPPPPPPP)"!(
    ComponentPose=@n3
    Alpha=1
    ConstraintWeights_0=0.0354
    ConstraintWeights_1:float=0.4161
    ConstraintWeights_2:float=0.0641
    ConstraintWeights_3:float=0.2293
    ConstraintWeights_4:float=0.0136
    ConstraintWeights_5:float=0.2328
    ConstraintWeights_6:float=0.0014
    ConstraintWeights_7:float=0.0073
   ) -> (Pose) {
    alphaBoolBlend {blendOption=Linear}
    boneToModify {boneName=const_buckle_back_r}
    constraintSetup [
     {targetBone {boneName=spine_02}}
     {targetBone {boneName=spine_04}}
     {targetBone {boneName=spine_03}}
     {targetBone {boneName=spine_05}}
     {targetBone {boneName=clavicle_l}}
     {targetBone {boneName=clavicle_r}}
     {targetBone {boneName=spine_01}}
     {targetBone {boneName=upperarm_twist_01_r}}
    ]
    constraintWeights [1,1,1,1,1,1,1,1]
   }
   n5 = RigidBody(ComponentPose=@n4, Alpha=Get SkirtAlpha()) {
    bClampLinearTranslationLimitToRefPose=true
    bOverrideWorldGravity=true
    overridePhysicsAsset=&/F_MED_Rays_Physics
    simulationSpace=BaseBoneSpace
    alphaBoolBlend {blendOption=Linear}
    alphaScaleBiasClamp {bInterpResult=true, interpSpeedDecreasing=5, interpSpeedIncreasing=5}
    baseBoneRef {boneName=pelvis}
    componentAppliedLinearAccClamp=<3000 3000 3000>
    componentLinearAccScale=<0.35 0.3 0.4>
    componentLinearVelScale=<0.35 0.3 0.4>
    overrideWorldGravity {z=-980}
    simSpaceSettings {
     maxAngularAcceleration=500
     maxAngularVelocity=3
     maxLinearAcceleration=400
     maxLinearVelocity=250
     velocityScaleZ=0
     worldAlpha=1
     externalLinearDragV=<0.1 0.1>
    }
   }
   n6 = Control Rig!(
    Source=Component To Local(ComponentPose=@n5)
    Alpha
    Crouching:bool=Get IsCrouching() -> (IsCrouching:bool)
    idlewithgun_alpha:double=Get idlewithgun_alpha() -> (idlewithgun_alpha:double)
    TacticalSprint_alpha:float=Get TacticalSprinting() -> (TacticalSprinting:double)
   ) -> (Pose) {
    alphaBoolBlend {blendOption=Linear}
    controlRigAssetReference {blueprintRigClass=&/F_MED_Rays_CtrlRig.F_MED_Rays_CtrlRig_C}
   }
   n7 = "Constraint - dyn_flap_frnt_parent_r (P)"(ComponentPose=Local To Component(LocalPose=@n6), Alpha=1, ConstraintWeights_0=1) {
    alphaBoolBlend {blendOption=Linear}
    boneToModify {boneName=dyn_flap_frnt_parent_r}
    constraintSetup [{targetBone {boneName=thigh_r}}]
    constraintWeights [1]
   }
   n8 = "Transform (Modify) Bone - Bone: dyn_flap_frnt_parent_r"!(ComponentPose=@n7, bAlphaBoolEnabled:bool=True, Translation="0,0,-1", Rotation="15,5,-20", Scale="1,1,1") -> (Pose) {
    alphaInputType=Bool
    rotationMode=BMM_Additive
    rotationSpace=BCS_BoneSpace
    scaleMode=BMM_Additive
    scaleSpace=BCS_BoneSpace
    translationMode=BMM_Additive
    translationSpace=BCS_BoneSpace
    alphaBoolBlend {blendOption=Linear}
    boneToModify {boneName=dyn_flap_frnt_parent_r}
    rotation=<rot 15 5 -20>
    translation {z=-1}
   }
   n9 = Look At - Bone: dyn_frill_side_l(ComponentPose=@n8, Alpha=1) {
    bUseLookUpAxis=true
    lookAtClamp=60
    alphaBoolBlend {blendOption=Linear}
    boneToModify {boneName=dyn_frill_side_l}
    lookAtLocation {x=0}
    lookAtTarget {boneReference {boneName=calf_l}}
    lookAt_Axis {axis=<1 0>}
    lookUp_Axis {axis {x=0, z=-1}}
   }
   n10 = Look At - Bone: dyn_frill_side_r(ComponentPose=@n9, Alpha=1) {
    bUseLookUpAxis=true
    lookAtClamp=60
    alphaBoolBlend {blendOption=Linear}
    boneToModify {boneName=dyn_frill_side_r}
    lookAtLocation {x=0}
    lookAtTarget {boneReference {boneName=calf_r}}
    lookAt_Axis {axis=<1 0>}
    lookUp_Axis {axis {x=0, z=-1}}
   }
   n11 = "Transform (Modify) Bone - Bone: driver_frill_side_r"!(ComponentPose=@n10, bAlphaBoolEnabled:bool=True, Translation="-7.247,0.599,1.123", Rotation="0.000,1.228,0.000", Scale="1,1,1") -> (Pose) {
    alphaInputType=Bool
    rotationMode=BMM_Additive
    rotationSpace=BCS_BoneSpace
    scaleMode=BMM_Additive
    scaleSpace=BCS_BoneSpace
    translationMode=BMM_Additive
    translationSpace=BCS_BoneSpace
    alphaBoolBlend {blendInTime=0.1, blendOption=Linear, blendOutTime=0.1}
    boneToModify {boneName=driver_frill_side_r}
    rotation=<rot 0 1.228 0>
    translation=<-7.247 0.599 1.123>
   }
   n12 = "Transform (Modify) Bone - Bone: driver_frill_bck_r"!(ComponentPose=@n11, bAlphaBoolEnabled:bool=True, Translation="-12.345,-0.094,-0.217", Rotation="1.076,0.000,22.446", Scale="1,1,1") -> (Pose) {
    alphaInputType=Bool
    rotationMode=BMM_Additive
    rotationSpace=BCS_BoneSpace
    scaleMode=BMM_Additive
    scaleSpace=BCS_BoneSpace
    translationMode=BMM_Additive
    translationSpace=BCS_BoneSpace
    alphaBoolBlend {blendInTime=0.1, blendOption=Linear, blendOutTime=0.1}
    boneToModify {boneName=driver_frill_bck_r}
    rotation {pitch=1.076, roll=22.446}
    translation=<-12.345 -0.094 -0.217>
   }
   n13 = "Transform (Modify) Bone - Bone: driver_frill_side_l"!(ComponentPose=@n12, bAlphaBoolEnabled:bool=True, Translation="-11.249,0.000,-0.000", Rotation="0,0,0", Scale="1,1,1") -> (Pose) {
    alphaInputType=Bool
    rotationMode=BMM_Additive
    rotationSpace=BCS_BoneSpace
    scaleMode=BMM_Additive
    scaleSpace=BCS_BoneSpace
    translationMode=BMM_Additive
    translationSpace=BCS_BoneSpace
    alphaBoolBlend {blendInTime=0.1, blendOption=Linear, blendOutTime=0.1}
    boneToModify {boneName=driver_frill_side_l}
    translation {x=-11.249}
   }
   n14 = "Transform (Modify) Bone - Bone: driver_frill_bck_l"!(ComponentPose=@n13, bAlphaBoolEnabled:bool=True, Translation="-10.083,0.000,0.000", Rotation="19.153,-0.000,0.000", Scale="1,1,1") -> (Pose) {
    alphaInputType=Bool
    rotationMode=BMM_Additive
    rotationSpace=BCS_BoneSpace
    scaleMode=BMM_Additive
    scaleSpace=BCS_BoneSpace
    translationMode=BMM_Additive
    translationSpace=BCS_BoneSpace
    alphaBoolBlend {blendInTime=0.1, blendOption=Linear, blendOutTime=0.1}
    boneToModify {boneName=driver_frill_bck_l}
    rotation=<rot 19.153 0 0>
    translation {x=-10.083}
   }
   n15 = "Transform (Modify) Bone - Bone: thigh_twist_01_r"!(ComponentPose=@n14, bAlphaBoolEnabled:bool=True, Translation="-0.440,-0.245,-1.576", Rotation="1.456,0.276,14.592", Scale="1,0.8,0.8") -> (Pose) {
    alphaInputType=Bool
    rotationMode=BMM_Additive
    rotationSpace=BCS_BoneSpace
    scaleMode=BMM_Additive
    scaleSpace=BCS_BoneSpace
    translationMode=BMM_Additive
    translationSpace=BCS_BoneSpace
    alphaBoolBlend {blendInTime=0.5, blendOption=Linear, blendOutTime=0.1}
    boneToModify {boneName=thigh_twist_01_r}
    rotation=<rot 1.456 0.276 14.5921>
    scale {y=0.8, z=0.8}
    translation=<-0.44 -0.245 -1.576>
   }
   n16 = "Transform (Modify) Bone - Bone: dyn_suspender_base_r"!(ComponentPose=@n15, bAlphaBoolEnabled:bool=True, Translation="-1.239,0.228,-1.122", Rotation="-0.000,-0.034,-11.586", Scale="1,1,1") -> (Pose) {
    alphaInputType=Bool
    rotationMode=BMM_Additive
    rotationSpace=BCS_BoneSpace
    scaleMode=BMM_Additive
    scaleSpace=BCS_BoneSpace
    translationMode=BMM_Additive
    translationSpace=BCS_BoneSpace
    alphaBoolBlend {blendInTime=0.5, blendOption=Linear, blendOutTime=0.1}
    boneToModify {boneName=dyn_suspender_base_r}
    rotation=<rot 0 -0.034 -11.586>
    translation=<-1.2394 0.228 -1.122>
   }
   n17 = "Transform (Modify) Bone - Bone: dyn_suspender_base_l"!(ComponentPose=@n16, bAlphaBoolEnabled:bool=True, Translation="0,0,0", Rotation="0,0,8", Scale="1,1,1") -> (Pose) {
    alphaInputType=Bool
    rotationMode=BMM_Additive
    rotationSpace=BCS_BoneSpace
    scaleMode=BMM_Additive
    scaleSpace=BCS_BoneSpace
    translationMode=BMM_Additive
    translationSpace=BCS_BoneSpace
    alphaBoolBlend {blendInTime=0.5, blendOption=Linear, blendOutTime=0.1}
    boneToModify {boneName=dyn_suspender_base_l}
    rotation {roll=8}
   }
   n18 = Get Slide() -> (Slide:double)
   n19 = Anim Dynamics - Bone: dyn_suspender_parent_r(ComponentPose=@n17, Alpha=@n18, ExternalForce="0,0,0") {
    angularDampingOverride=0.9999
    angularSpringConstant=10000
    bAngularSpring=true
    bOverrideAngularDamping=true
    bOverrideLinearDamping=true
    bUseSphericalLimits=true
    linearDampingOverride=0.9999
    simulationSpace=RootRelative
    alphaBoolBlend {blendOption=Linear}
    boundBone {boneName=dyn_suspender_parent_r}
    componentAppliedLinearAccClamp=<6000 6000 6000>
    componentLinearAccScale=<0.3 0.3 0.3>
    componentLinearVelScale=<0.3 0.3 0.3>
    physicsBodyDefinitions [{
      sphereCollisionRadius=5
      boundBone {boneName=dyn_suspender_parent_r}
      boxExtents=<10 10 10>
      constraintSetup {angularTargetAxis=AxisY, angularLimitsMax=<10 10 10>, angularLimitsMin=<-10 0 -10>, angularTarget=<1 0 0>}
      localJointOffset=<1 0 0>
     }]
    sphericalLimits [{limitRadius=5, drivingBone {boneName=spine_02}, sphereLocalOffset=<0 0 0>}]
   }
   n20 = Anim Dynamics - Bone: dyn_suspender_parent_l(ComponentPose=@n19, Alpha=@n18, ExternalForce="0,0,0") {
    angularDampingOverride=0.9999
    angularSpringConstant=10000
    bAngularSpring=true
    bOverrideAngularDamping=true
    bOverrideLinearDamping=true
    bUsePlanarLimit=false
    bUseSphericalLimits=true
    linearDampingOverride=0.9999
    simulationSpace=RootRelative
    alphaBoolBlend {blendOption=Linear}
    boundBone {boneName=dyn_suspender_parent_l}
    componentAppliedLinearAccClamp=<6000 6000 6000>
    componentLinearAccScale=<0.3 0.3 0.3>
    componentLinearVelScale=<0.3 0.3 0.3>
    physicsBodyDefinitions [{
      collisionType=CustomSphere
      sphereCollisionRadius=5
      boundBone {boneName=dyn_suspender_parent_l}
      boxExtents=<20 10 10>
      constraintSetup {angularTargetAxis=AxisY, angularLimitsMax=<0 30 10>, angularLimitsMin=<-30 0 -10>, angularTarget=<1 0 0>}
      localJointOffset=<0.3 0 0>
     }]
    sphericalLimits [{limitRadius=6, drivingBone {boneName=spine_02}, sphereLocalOffset=<3 0 0>}]
   }
   n21 = Anim Dynamics - Bone: dyn_suspender_02_r(ComponentPose=@n20, Alpha=@n18, ExternalForce="0,0,0") {
    angularDampingOverride=0.9999
    angularSpringConstant=10000
    bAngularSpring=true
    bOverrideAngularDamping=true
    bOverrideLinearDamping=true
    bUseSphericalLimits=true
    linearDampingOverride=0.9999
    simulationSpace=RootRelative
    alphaBoolBlend {blendOption=Linear}
    boundBone {boneName=dyn_suspender_02_r}
    componentAppliedLinearAccClamp=<6000 6000 6000>
    componentLinearAccScale=<0.3 0.3 0.3>
    componentLinearVelScale=<0.3 0.3 0.3>
    physicsBodyDefinitions [{
      collisionType=CustomSphere
      sphereCollisionRadius=7
      boundBone {boneName=dyn_suspender_02_r}
      boxExtents=<10 10 10>
      constraintSetup {angularTargetAxis=AxisY, angularLimitsMax=<5 5 5>, angularLimitsMin=<-5 0 -5>, angularTarget=<1 0 0>}
      localJointOffset=<0.5 0 0>
     }]
    sphericalLimits [{limitRadius=5, drivingBone {boneName=spine_02}, sphereLocalOffset=<0 0 0>}]
   }
   n22 = Anim Dynamics - Bone: dyn_suspender_01_l(ComponentPose=@n21, Alpha=@n18, ExternalForce="0,0,0") {
    angularDampingOverride=0.9999
    angularSpringConstant=10000
    bAngularSpring=true
    bOverrideAngularDamping=true
    bOverrideLinearDamping=true
    bUsePlanarLimit=false
    bUseSphericalLimits=true
    linearDampingOverride=0.9999
    simulationSpace=RootRelative
    alphaBoolBlend {blendOption=Linear}
    boundBone {boneName=dyn_suspender_01_l}
    componentAppliedLinearAccClamp=<6000 6000 6000>
    componentLinearAccScale=<0.3 0.3 0.3>
    componentLinearVelScale=<0.3 0.3 0.3>
    physicsBodyDefinitions [{
      collisionType=CustomSphere
      sphereCollisionRadius=8
      boundBone {boneName=dyn_suspender_01_l}
      boxExtents=<10 10 10>
      constraintSetup {angularTargetAxis=AxisY, angularLimitsMax=<0 5 5>, angularLimitsMin=<0 0 -5>, angularTarget=<1 0 0>}
      localJointOffset=<0.1 0 0>
     }]
    sphericalLimits [{limitRadius=6, drivingBone {boneName=spine_02}, sphereLocalOffset=<0 0 0>}]
   }
   Output Pose(Result=Component To Local(ComponentPose=@n22)) {layerGroup=None}
  ]}
 EventGraph {nodes [
   Event BlueprintUpdateAnimation()
   > n1 = Sequence!(execute) -> (then_0, then_1, then_2:Exec, then_3:Exec, then_4:Exec, then_5:Exec)
   > Set Slide(Slide:double=float - float(A=1, B=Get SkirtAlpha())) -> (Output_Get:double)
   n2 = GetOwnerSkeletalMeshComponent()
   Set legSwingDelta_r(
    execute=@n1.then_1
    legSwingDelta_r:double=MapRangeClamped(Value=GetTwistAndSwingAngleOfDeltaRotationFromRefPose(self=@n2, BoneName=thigh_r).OutSwingAngle, InRangeA=150, InRangeB=75)
   ) -> (Output_Get:double)
   > Set legSwingDelta_l(legSwingDelta_l:double=MapRangeClamped(Value=GetTwistAndSwingAngleOfDeltaRotationFromRefPose(self=@n2, BoneName=thigh_l).OutSwingAngle, InRangeA=43, InRangeB=110)) -> (Output_Get:double)
   n3 = Branch(execute=@n1.then_3, Condition=Get bIsSliding())
   > Set SkirtAlpha(SkirtAlpha:float=0) -> (Output_Get:float)
   Set SkirtAlpha(execute=@n3.else, SkirtAlpha=1)
   n4 = Get WeaponCoreAnim!() -> (WeaponCoreAnim)
   Set idlewithgun_alpha(
    execute=@n1.then_4
    idlewithgun_alpha:double="To Float (Boolean)"(InBool="Equal (Enum)"(A:EFortWeaponCoreAnimation Enum=@n4, B:EFortWeaponCoreAnimation Enum=Pistol))
   ) -> (Output_Get:double)
   n5 = Get bIsTacticalSprinting()
   Set TacticalSprinting(
    execute=@n1.then_5
    TacticalSprinting:double=FInterpTo(
     Current=Get TacticalSprinting()
     Target="To Float (Boolean)"(InBool=@n5)
     DeltaTime=GetWorldDeltaSeconds!() -> (ReturnValue:double)
     InterpSpeed=20
    )
   ) -> (Output_Get:double)
   n6 = Branch(execute=@n1.then_2, Condition=Get bIsCrouching())
   > Set IsCrouching(IsCrouching:bool=true) -> (Output_Get:bool)
   Set IsCrouching(execute=@n6.else, IsCrouching=false)
   SelectFloat(A=1, B=0, bPickA=Get bIsCrouching())
  ]}
]
variables [
 idlewithgun_alpha {type=number}
 IsCrouching {type=boolean}
 legSwingDelta_l {type=number}
 legSwingDelta_r {type=number}
 SkirtAlpha {type=number, default=1}
 Slide {type=number}
 TacticalSprinting {type=number}
]
