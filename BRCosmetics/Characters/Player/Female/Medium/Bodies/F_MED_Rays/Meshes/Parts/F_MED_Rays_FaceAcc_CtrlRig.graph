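# Control Rig for the F_MED_Rays face accessory that procedurally poses the hair base and braid bones to follow the character's pose.
# Spherical pose readers on the head blend corrective offsets, including one complementary pair.
# When `aim` is true, right upper-arm and head readers combine to weight an extra hair-base nudge.
# Constraint, hair-aim and cape-dependent braid scaling chains exist but are disconnected from execution.
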ControlRigBlueprint /BRCosmetics/Characters/Player/Female/Medium/Bodies/F_MED_Rays/Meshes/Parts/F_MED_Rays_FaceAcc_CtrlRig
graphs [
 Multiply_1_ContainedGraph {nodes [
   n1 = RigVMFunctionEntryNode()
   Multiply = &RigVMFunction_MathDoubleMul(A=@n1.A, B=@n1.B)
   RigVMFunctionReturnNode(Result=&RigVMFunction_MathDoubleMul(B=@n1.C, A=@Multiply.Result).Result)
  ]}
 RigVMFunctionLibrary {links [], nodes []}
 RigVMModel {nodes [
   n1 = &RigUnit_GetRelativeTransformForItem(Child=(Type=Bone,Name="head"), Parent=(Type=Bone,Name="spine_05"), bChildInitial=false)
   &RigUnit_ModifyTransforms(
    ExecutePin=()
    ItemToModify=((Item=(Type=Bone,Name="dyn_hair_aim"),Transform=(Rotation=(X=0,Y=0,Z=0,W=1),Translation=(X=0,Y=15,Z=20),Scale3D=(X=1,Y=1,Z=1))))
    Weight=&RigVMFunction_MathFloatRemap(
     SourceMaximum=0
     SourceMinimum=-45
     TargetMaximum=0
     TargetMinimum=1
     bClamp=true
     Value=&RigVMFunction_MathQuaternionToEuler(RotationOrder=ZYX, Value=@n1.RelativeTransform.Rotation).Result.X
    ).Result
   )
   n2 = &RigUnit_GetRelativeTransformForItem(Child=(Type=Bone,Name="head"), Parent=(Type=Bone,Name="spine_05"))
   > &RigUnit_ModifyTransforms(
    ItemToModify=((Item=(Type=Bone,Name="dyn_hair_aim"),Transform=(Rotation=(X=0,Y=0,Z=0,W=1),Translation=(X=0,Y=15,Z=-20),Scale3D=(X=1,Y=1,Z=1))))
    Weight=&RigVMFunction_MathFloatRemap(
     SourceMaximum=45
     SourceMinimum=0
     TargetMinimum=0
     bClamp=true
     Value=&RigVMFunction_MathQuaternionToEuler(RotationOrder=ZYX, Value=@n2.RelativeTransform.Rotation).Result.X
    ).Result
   )
   &RigUnit_ModifyTransforms(
    ExecutePin=()
    ItemToModify=((Item=(Type=Bone,Name="dyn_braid_02_a"),Transform=(Rotation=(X=0,Y=0,Z=0,W=1),Translation=(X=0,Y=0,Z=0),Scale3D=(X=0.75,Y=0.75,Z=0.75))))
    Weight=1
   )
   > &RigUnit_ModifyTransforms(
    ItemToModify=((Item=(Type=Bone,Name="dyn_braid_02_b"),Transform=(Rotation=(X=0,Y=0,Z=0,W=1),Translation=(X=0,Y=0,Z=0),Scale3D=(X=0.75,Y=0.75,Z=0.75))))
    Weight=1
   )
   > &RigUnit_ModifyTransforms(
    ItemToModify=((Item=(Type=Bone,Name="dyn_braid_02_c"),Transform=(Rotation=(X=0,Y=0,Z=0,W=1),Translation=(X=0,Y=0,Z=0),Scale3D=(X=0.75,Y=0.75,Z=0.75))))
    Weight=1
   )
   > &RigUnit_ModifyTransforms(
    ItemToModify=((Item=(Type=Bone,Name="dyn_braid_02_d"),Transform=(Rotation=(X=0,Y=0,Z=0,W=1),Translation=(X=0,Y=0,Z=0),Scale3D=(X=0.75,Y=0.75,Z=0.75))))
    Weight=1
   )
   > &RigUnit_ModifyTransforms(
    ItemToModify=((Item=(Type=Bone,Name="dyn_braid_02_e"),Transform=(Rotation=(X=0,Y=0,Z=0,W=1),Translation=(X=0,Y=0,Z=0),Scale3D=(X=0.75,Y=0.75,Z=0.75))))
    Weight=1
   )
   > &RigUnit_ModifyTransforms(
    ItemToModify=((Item=(Type=Bone,Name="dyn_braid_02_f"),Transform=(Rotation=(X=0,Y=0,Z=0,W=1),Translation=(X=0,Y=0,Z=0),Scale3D=(X=0.75,Y=0.75,Z=0.75))))
    Weight=1
   )
   &RigUnit_ParentConstraint(
    AdvancedSettings=(InterpolationType=Average,RotationOrderForFilter=XZY)
    Child=(Type=Bone,Name="dyn_braid_02_a")
    ExecutePin=()
    Parents=((Item=(Type=Bone,Name="dyn_hair_aim"),Weight=1))
    Weight=1
   )
   > &RigUnit_ParentConstraint(
    AdvancedSettings=(InterpolationType=Average,RotationOrderForFilter=XZY)
    Child=(Type=Bone,Name="dyn_braid_02_b")
    Parents=((Item=(Type=Bone,Name="dyn_hair_aim"),Weight=1))
    Weight=1
   )
   > &RigUnit_ParentConstraint(
    AdvancedSettings=(InterpolationType=Average,RotationOrderForFilter=XZY)
    Child=(Type=Bone,Name="dyn_braid_02_c")
    Parents=((Item=(Type=Bone,Name="dyn_hair_aim"),Weight=1))
    Weight=1
   )
   > &RigUnit_ParentConstraint(
    AdvancedSettings=(InterpolationType=Average,RotationOrderForFilter=XZY)
    Child=(Type=Bone,Name="dyn_braid_02_d")
    Parents=((Item=(Type=Bone,Name="dyn_hair_aim"),Weight=1))
    Weight=1
   )
   > &RigUnit_ParentConstraint(
    AdvancedSettings=(InterpolationType=Average,RotationOrderForFilter=XZY)
    Child=(Type=Bone,Name="dyn_braid_02_e")
    Filter=(TranslationFilter=(bX=false,bY=True,bZ=True),RotationFilter=(bX=True,bY=True,bZ=True),ScaleFilter=(bX=True,bY=True,bZ=True))
    Parents=((Item=(Type=Bone,Name="dyn_hair_aim"),Weight=1))
    Weight=1
   )
   > &RigUnit_ParentConstraint(
    AdvancedSettings=(InterpolationType=Average,RotationOrderForFilter=XZY)
    Child=(Type=Bone,Name="dyn_braid_02_f")
    Filter=(TranslationFilter=(bX=false,bY=True,bZ=True),RotationFilter=(bX=True,bY=True,bZ=True),ScaleFilter=(bX=True,bY=True,bZ=True))
    Parents=((Item=(Type=Bone,Name="dyn_hair_aim"),Weight=1))
    Weight=1
   )
   /Script/ControlRig.RigUnit_PositionConstraintLocalSpaceOffset(
    Child=(Type=Bone,Name="dyn_braid_02_a")
    ExecutePin=()
    Filter=(bX=True,bY=True,bZ=True)
    Parents=((Item=(Type=Bone,Name="dyn_hair_aim"),Weight=1))
    Weight=1
    bMaintainOffset=True
   ) {is=struct RigVMUnitNode}
   > &RigUnit_PositionConstraintLocalSpaceOffset(
    Child=(Type=Bone,Name="dyn_braid_02_b")
    Filter=(bX=True,bY=True,bZ=True)
    Parents=((Item=(Type=Bone,Name="dyn_hair_aim"),Weight=1))
    Weight=1
    bMaintainOffset=True
   )
   > &RigUnit_PositionConstraintLocalSpaceOffset(
    Child=(Type=Bone,Name="dyn_braid_02_c")
    Filter=(bX=True,bY=True,bZ=True)
    Parents=((Item=(Type=Bone,Name="dyn_hair_aim"),Weight=1))
    Weight=1
    bMaintainOffset=True
   )
   > &RigUnit_PositionConstraintLocalSpaceOffset(
    Child=(Type=Bone,Name="dyn_braid_02_d")
    Filter=(bX=True,bY=True,bZ=True)
    Parents=((Item=(Type=Bone,Name="dyn_hair_aim"),Weight=1))
    Weight=1
    bMaintainOffset=True
   )
   > &RigUnit_PositionConstraintLocalSpaceOffset(
    Child=(Type=Bone,Name="dyn_braid_02_e")
    Filter=(bX=True,bY=True,bZ=True)
    Parents=((Item=(Type=Bone,Name="dyn_hair_aim"),Weight=1))
    Weight=1
    bMaintainOffset=True
   )
   > &RigUnit_PositionConstraintLocalSpaceOffset(
    Child=(Type=Bone,Name="dyn_braid_02_f")
    Filter=(bX=True,bY=True,bZ=True)
    Parents=((Item=(Type=Bone,Name="dyn_hair_aim"),Weight=1))
    Weight=1
    bMaintainOffset=True
   )
   &RigUnit_PositionConstraintLocalSpaceOffset(
    Child=(Type=Bone,Name="dyn_braid_01_a")
    ExecutePin=()
    Filter=(bX=True,bY=True,bZ=True)
    Parents=((Item=(Type=Bone,Name="spine_05"),Weight=1))
    Weight=0.5
    bMaintainOffset=True
   )
   > &RigUnit_PositionConstraintLocalSpaceOffset(
    Child=(Type=Bone,Name="dyn_braid_01_b")
    Filter=(bX=True,bY=True,bZ=True)
    Parents=((Item=(Type=Bone,Name="spine_05"),Weight=1))
    Weight=1
    bMaintainOffset=True
   )
   > &RigUnit_PositionConstraintLocalSpaceOffset(
    Child=(Type=Bone,Name="dyn_braid_01_c")
    Filter=(bX=True,bY=True,bZ=True)
    Parents=((Item=(Type=Bone,Name="spine_05"),Weight=1))
    Weight=1
    bMaintainOffset=True
   )
   > &RigUnit_PositionConstraintLocalSpaceOffset(
    Child=(Type=Bone,Name="dyn_braid_01_d")
    Filter=(bX=True,bY=True,bZ=True)
    Parents=((Item=(Type=Bone,Name="spine_05"),Weight=1))
    Weight=1
    bMaintainOffset=True
   )
   > &RigUnit_PositionConstraintLocalSpaceOffset(
    Child=(Type=Bone,Name="dyn_braid_01_e")
    Filter=(bX=True,bY=True,bZ=True)
    Parents=((Item=(Type=Bone,Name="spine_05"),Weight=1))
    Weight=1
    bMaintainOffset=True
   )
   > &RigUnit_PositionConstraintLocalSpaceOffset(
    Child=(Type=Bone,Name="dyn_braid_01_f")
    Filter=(bX=True,bY=True,bZ=True)
    Parents=((Item=(Type=Bone,Name="spine_05"),Weight=1))
    Weight=1
    bMaintainOffset=True
   )
   RigUnit_BeginExecution = &RigUnit_BeginExecution()
   RigVMFunction_Sequence = &RigVMFunction_Sequence(ExecuteContext=@RigUnit_BeginExecution.ExecutePin)
   n3 = &RigUnit_SphericalPoseReader(
    ActiveRegionScaleFactors=(PositiveWidth=1,NegativeWidth=1,PositiveHeight=1,NegativeHeight=1)
    ActiveRegionSize=0.01
    Debug=(bDrawDebug=true,bDraw2D=False,bDrawLocalAxes=False,DebugScale=25,DebugSegments=20,DebugThickness=0.25)
    DriverAxis=(X=1,Y=0,Z=0)
    DriverItem=(Type=Bone,Name="head")
    FalloffRegionScaleFactors=(PositiveWidth=1,NegativeWidth=1,PositiveHeight=1,NegativeHeight=1)
    FalloffSize=0.4
    FlipHeightScaling=False
    FlipWidthScaling=False
    OptionalParentItem=(Type=Bone,Name="None")
    RotationOffset=(X=20,Y=-100,Z=0)
    ExecutePin=@RigVMFunction_Sequence.A
   )
   RigVMFunction_MathQuaternionFromEuler = &RigVMFunction_MathQuaternionFromEuler(Euler=(X=0,Y=0,Z=0))
   n4 = &RigVMFunction_MathQuaternionFromEuler(Euler=(X=0,Y=13,Z=0))
   n5 = &RigVMFunction_MathQuaternionFromEuler(Euler=(X=0,Y=0,Z=0))
   n6 = &RigVMFunction_MathQuaternionFromEuler(Euler=(X=0,Y=-10,Z=0))
   > &RigUnit_ModifyTransforms(
    ItemToModify="((Item=(Type=Bone,Name=\"dyn_hair_base\"),Transform=(Rotation=(X=0,Y=0,Z=0,W=1),Translation=(X=0,Y=0,Z=-0.6353),Scale3D=(X=1,Y=1,Z=1))),(Item=(Type=Bone,Name=\"dyn_braid_base_a\"),Transform=(Rotation=(X=0,Y=0,Z=0,W=1),Translation=(X=0,Y=0,Z=0),Scale3D=(X=1,Y=1 ...(truncated)"
    ItemToModify.0.Transform.Rotation=@RigVMFunction_MathQuaternionFromEuler.Result
    ItemToModify.1.Transform.Rotation=@n4.Result
    ItemToModify.2.Transform.Rotation=@n4.Result
    ItemToModify.3.Transform.Rotation=@n4.Result
    ItemToModify.4.Transform.Rotation=@n4.Result
    ItemToModify.5.Transform.Rotation=@n4.Result
    ItemToModify.6.Transform.Rotation=@n4.Result
    ItemToModify.10.Transform.Rotation=@n5.Result
    ItemToModify.11.Transform.Rotation=@n5.Result
    ItemToModify.12.Transform.Rotation=@n5.Result
    ItemToModify.7.Transform.Rotation=@n5.Result
    ItemToModify.8.Transform.Rotation=@n5.Result
    ItemToModify.9.Transform.Rotation=@n5.Result
    ItemToModify.13.Transform.Rotation=@n6.Result
    ItemToModify.14.Transform.Rotation=@n6.Result
    ItemToModify.15.Transform.Rotation=@n6.Result
    ItemToModify.16.Transform.Rotation=@n6.Result
    ItemToModify.17.Transform.Rotation=@n6.Result
    ItemToModify.18.Transform.Rotation=@n6.Result
    Weight=@n3.OutputParam
   )
   n7 = &RigVMFunction_MathQuaternionFromEuler(Euler=(X=0,Y=0,Z=0))
   n8 = &RigVMFunction_MathQuaternionFromEuler(Euler=(X=0,Y=0,Z=0))
   n9 = &RigVMFunction_MathQuaternionFromEuler(Euler=(X=0,Y=13,Z=0))
   Subtract = &RigVMFunction_MathDoubleSub(A=1, B=@n3.OutputParam)
   > &RigUnit_ModifyTransforms(
    ItemToModify="((Item=(Type=Bone,Name=\"dyn_hair_base\"),Transform=(Rotation=(X=0,Y=0,Z=0,W=1),Translation=(X=-0.1391,Y=0,Z=8.3416),Scale3D=(X=1,Y=1,Z=1))),(Item=(Type=Bone,Name=\"dyn_braid_base_a\"),Transform=(Rotation=(X=0,Y=0,Z=0,W=1),Translation=(X=0,Y=0,Z=0),Scale3D=(X= ...(truncated)"
    ItemToModify.13.Transform.Rotation=@n7.Result
    ItemToModify.14.Transform.Rotation=@n7.Result
    ItemToModify.15.Transform.Rotation=@n7.Result
    ItemToModify.16.Transform.Rotation=@n7.Result
    ItemToModify.17.Transform.Rotation=@n7.Result
    ItemToModify.18.Transform.Rotation=@n7.Result
    ItemToModify.10.Transform.Rotation=@n8.Result
    ItemToModify.11.Transform.Rotation=@n8.Result
    ItemToModify.12.Transform.Rotation=@n8.Result
    ItemToModify.7.Transform.Rotation=@n8.Result
    ItemToModify.8.Transform.Rotation=@n8.Result
    ItemToModify.9.Transform.Rotation=@n8.Result
    ItemToModify.1.Transform.Rotation=@n9.Result
    ItemToModify.2.Transform.Rotation=@n9.Result
    ItemToModify.3.Transform.Rotation=@n9.Result
    ItemToModify.4.Transform.Rotation=@n9.Result
    ItemToModify.5.Transform.Rotation=@n9.Result
    ItemToModify.6.Transform.Rotation=@n9.Result
    ItemToModify.0.Transform.Rotation=&RigVMFunction_MathQuaternionFromEuler(Euler=(X=0,Y=-10,Z=0)).Result
    Weight=@Subtract.Result
   )
   > n10 = &RigUnit_SphericalPoseReader(
    ActiveRegionScaleFactors=(PositiveWidth=1,NegativeWidth=1,PositiveHeight=1,NegativeHeight=1)
    ActiveRegionSize=0.01
    Debug=(bDrawDebug=false,bDraw2D=False,bDrawLocalAxes=False,DebugScale=25,DebugSegments=20,DebugThickness=0.25)
    DriverAxis=(X=1,Y=0,Z=0)
    DriverItem=(Type=Bone,Name="head")
    FalloffRegionScaleFactors=(PositiveWidth=1,NegativeWidth=1,PositiveHeight=0.6,NegativeHeight=0.8)
    FalloffSize=0.3
    FlipHeightScaling=False
    FlipWidthScaling=False
    OptionalParentItem=(Type=Bone,Name="None")
    RotationOffset=(X=0,Y=-90,Z=0)
   )
   n11 = &RigVMFunction_MathQuaternionFromEuler(Euler=(X=0,Y=0,Z=0))
   n12 = &RigVMFunction_MathQuaternionFromEuler(Euler=(X=0,Y=5,Z=0))
   n13 = &RigVMFunction_MathQuaternionFromEuler(Euler=(X=0,Y=5,Z=0))
   > &RigUnit_ModifyTransforms(
    ItemToModify="((Item=(Type=Bone,Name=\"dyn_hair_base\"),Transform=(Rotation=(X=0,Y=0,Z=0,W=1),Translation=(X=0.4508,Y=0,Z=0.7282),Scale3D=(X=1,Y=1,Z=1))),(Item=(Type=Bone,Name=\"dyn_braid_base_a\"),Transform=(Rotation=(X=0,Y=0,Z=0,W=1),Translation=(X=0,Y=0,Z=0),Scale3D=(X=1 ...(truncated)"
    ItemToModify.13.Transform.Rotation=@n11.Result
    ItemToModify.14.Transform.Rotation=@n11.Result
    ItemToModify.15.Transform.Rotation=@n11.Result
    ItemToModify.16.Transform.Rotation=@n11.Result
    ItemToModify.17.Transform.Rotation=@n11.Result
    ItemToModify.18.Transform.Rotation=@n11.Result
    ItemToModify.10.Transform.Rotation=@n12.Result
    ItemToModify.11.Transform.Rotation=@n12.Result
    ItemToModify.12.Transform.Rotation=@n12.Result
    ItemToModify.7.Transform.Rotation=@n12.Result
    ItemToModify.8.Transform.Rotation=@n12.Result
    ItemToModify.9.Transform.Rotation=@n12.Result
    ItemToModify.1.Transform.Rotation=@n13.Result
    ItemToModify.2.Transform.Rotation=@n13.Result
    ItemToModify.3.Transform.Rotation=@n13.Result
    ItemToModify.4.Transform.Rotation=@n13.Result
    ItemToModify.5.Transform.Rotation=@n13.Result
    ItemToModify.6.Transform.Rotation=@n13.Result
    ItemToModify.0.Transform.Rotation=&RigVMFunction_MathQuaternionFromEuler(Euler=(X=0,Y=-10,Z=0)).Result
    Weight=@n10.OutputParam
   )
   > n14 = &RigUnit_SphericalPoseReader(
    ActiveRegionScaleFactors=(PositiveWidth=1,NegativeWidth=1,PositiveHeight=1,NegativeHeight=1)
    ActiveRegionSize=0.02
    Debug=(bDrawDebug=true,bDraw2D=False,bDrawLocalAxes=False,DebugScale=25,DebugSegments=20,DebugThickness=0.25)
    DriverAxis=(X=1,Y=0,Z=0)
    DriverItem=(Type=Bone,Name="head")
    FalloffRegionScaleFactors=(PositiveWidth=1,NegativeWidth=1,PositiveHeight=1,NegativeHeight=1)
    FalloffSize=0.1
    FlipHeightScaling=False
    FlipWidthScaling=False
    OptionalParentItem=(Type=Bone,Name="None")
    RotationOffset=(X=-45,Y=-60,Z=0)
   )
   n15 = &RigVMFunction_MathQuaternionFromEuler(Euler=(X=0,Y=0,Z=0))
   n16 = &RigVMFunction_MathQuaternionFromEuler(Euler=(X=0,Y=-5,Z=0))
   n17 = &RigVMFunction_MathQuaternionFromEuler(Euler=(X=0,Y=0,Z=0))
   > &RigUnit_ModifyTransforms(
    ItemToModify="((Item=(Type=Bone,Name=\"dyn_hair_base\"),Transform=(Rotation=(X=0,Y=0,Z=0,W=1),Translation=(X=-1.7875,Y=0.0141,Z=-6.3693),Scale3D=(X=1,Y=1,Z=1))),(Item=(Type=Bone,Name=\"dyn_braid_base_a\"),Transform=(Rotation=(X=0,Y=0,Z=0,W=1),Translation=(X=0,Y=0,Z=0),Scale ...(truncated)"
    ItemToModify.13.Transform.Rotation=@n15.Result
    ItemToModify.14.Transform.Rotation=@n15.Result
    ItemToModify.15.Transform.Rotation=@n15.Result
    ItemToModify.16.Transform.Rotation=@n15.Result
    ItemToModify.17.Transform.Rotation=@n15.Result
    ItemToModify.18.Transform.Rotation=@n15.Result
    ItemToModify.10.Transform.Rotation=@n16.Result
    ItemToModify.11.Transform.Rotation=@n16.Result
    ItemToModify.12.Transform.Rotation=@n16.Result
    ItemToModify.7.Transform.Rotation=@n16.Result
    ItemToModify.8.Transform.Rotation=@n16.Result
    ItemToModify.9.Transform.Rotation=@n16.Result
    ItemToModify.1.Transform.Rotation=@n17.Result
    ItemToModify.2.Transform.Rotation=@n17.Result
    ItemToModify.3.Transform.Rotation=@n17.Result
    ItemToModify.4.Transform.Rotation=@n17.Result
    ItemToModify.5.Transform.Rotation=@n17.Result
    ItemToModify.6.Transform.Rotation=@n17.Result
    ItemToModify.0.Transform.Rotation=&RigVMFunction_MathQuaternionFromEuler(Euler=(X=0,Y=0,Z=0)).Result
    Weight=@n14.OutputParam
   )
   n18 = &RigUnit_SphericalPoseReader(
    ActiveRegionScaleFactors=(PositiveWidth=1,NegativeWidth=1,PositiveHeight=1,NegativeHeight=1)
    ActiveRegionSize=0.001
    Debug=(bDrawDebug=true,bDraw2D=False,bDrawLocalAxes=False,DebugScale=25,DebugSegments=20,DebugThickness=0.25)
    DriverAxis=(X=-1,Y=0,Z=0)
    DriverItem=(Type=Bone,Name="upperarm_r")
    FalloffRegionScaleFactors=(PositiveWidth=1,NegativeWidth=1,PositiveHeight=1,NegativeHeight=1)
    FalloffSize=0.3
    OptionalParentItem=(Type=Bone)
    RotationOffset=(X=-131.1485,Y=-25.4886,Z=25.1809)
    ExecutePin=@RigVMFunction_Sequence.B
   )
   > n19 = &RigUnit_SphericalPoseReader(
    ActiveRegionScaleFactors=(PositiveWidth=1,NegativeWidth=1,PositiveHeight=1,NegativeHeight=1)
    ActiveRegionSize=0.05
    Debug=(bDrawDebug=true,bDraw2D=False,bDrawLocalAxes=False,DebugScale=25,DebugSegments=20,DebugThickness=0.25)
    DriverAxis=(X=0,Y=-1,Z=0)
    DriverItem=(Type=Bone,Name="upperarm_r")
    FalloffRegionScaleFactors=(PositiveWidth=1,NegativeWidth=1,PositiveHeight=1,NegativeHeight=1)
    FalloffSize=0.3
    OptionalParentItem=(Type=Bone)
    RotationOffset=(X=-114.8871,Y=6.3019,Z=46.6476)
   )
   > n20 = &RigUnit_SphericalPoseReader(
    ActiveRegionScaleFactors=(PositiveWidth=1,NegativeWidth=0.5,PositiveHeight=0.75,NegativeHeight=0.75)
    ActiveRegionSize=0.03
    Debug=(bDrawDebug=true,bDraw2D=False,bDrawLocalAxes=False,DebugScale=25,DebugSegments=20,DebugThickness=0.25)
    DriverAxis=(X=0,Y=1,Z=0)
    DriverItem=(Type=Bone,Name="head")
    FalloffRegionScaleFactors=(PositiveWidth=1,NegativeWidth=0.5,PositiveHeight=1,NegativeHeight=1)
    FalloffSize=0.3
    OptionalParentItem=(Type=Bone)
    RotationOffset=(X=97.7348,Y=0.3762,Z=-22.3624)
   )
   Maximum = &RigVMFunction_MathDoubleMax(B=&RigVMFunction_MathDoubleMul(B=0.2, A=@n18.OutputParam).Result, A=@n19.OutputParam)
   RigVMFunction_MathBoolToFloat = &RigVMFunction_MathBoolToFloat(Value=RigVMVariableNode() {variable=aim}.Value)
   Multiply = RigVMAggregateNode(C=@Maximum.Result, A=@RigVMFunction_MathBoolToFloat.Result, B=@n20.OutputParam)
   > &RigUnit_ModifyTransforms(
    ItemToModify=((Item=(Type=Bone,Name="dyn_hair_base"),Transform=(Rotation=(X=0,Y=0,Z=0,W=1),Translation=(X=0,Y=0,Z=-5),Scale3D=(X=1,Y=1,Z=1))))
    ItemToModify.0.Transform.Rotation=&RigVMFunction_MathQuaternionFromEuler(Euler=(X=0,Y=5,Z=0)).Result
    Weight=@Multiply.Result
   )
   n21 = &RigUnit_SetTransform(
    ExecutePin=()
    Item=(Type=Bone,Name="dyn_hair_aim")
    Space=LocalSpace
    Value=(Rotation=(X=0,Y=0,Z=0,W=1),Translation=(X=0,Y=-13,Z=0),Scale3D=(X=1,Y=1,Z=1))
    Weight=1
    bInitial=false
   )
   Branch_RigVMUnitNode = &RigVMFunction_ControlFlowBranch(ExecuteContext=@n21.ExecutePin, Condition=RigVMVariableNode() {variable=isWearingCape}.Value)
   &RigUnit_ModifyTransforms(
    ItemToModify=((Item=(Type=Bone,Name="dyn_braid_01_a"),Transform=(Rotation=(X=0,Y=0,Z=0,W=1),Translation=(X=0,Y=0,Z=0),Scale3D=(X=0.8,Y=0.8,Z=0.8))))
    Weight=1
    ExecutePin=@Branch_RigVMUnitNode.True
   )
   > &RigUnit_ModifyTransforms(
    ItemToModify=((Item=(Type=Bone,Name="dyn_braid_01_b"),Transform=(Rotation=(X=0,Y=0,Z=0,W=1),Translation=(X=0,Y=0,Z=0),Scale3D=(X=0.8,Y=0.8,Z=0.8))))
    Weight=1
   )
   > &RigUnit_ModifyTransforms(
    ItemToModify=((Item=(Type=Bone,Name="dyn_braid_01_c"),Transform=(Rotation=(X=0,Y=0,Z=0,W=1),Translation=(X=0,Y=0,Z=0),Scale3D=(X=0.8,Y=0.8,Z=0.8))))
    Weight=1
   )
   > &RigUnit_ModifyTransforms(
    ItemToModify=((Item=(Type=Bone,Name="dyn_braid_01_d"),Transform=(Rotation=(X=0,Y=0,Z=0,W=1),Translation=(X=0,Y=0,Z=0),Scale3D=(X=0.8,Y=0.8,Z=0.8))))
    Weight=1
   )
   > &RigUnit_ModifyTransforms(
    ItemToModify=((Item=(Type=Bone,Name="dyn_braid_01_e"),Transform=(Rotation=(X=0,Y=0,Z=0,W=1),Translation=(X=0,Y=0,Z=0),Scale3D=(X=0.8,Y=0.8,Z=0.8))))
    Weight=1
   )
   > &RigUnit_ModifyTransforms(
    ItemToModify=((Item=(Type=Bone,Name="dyn_braid_01_f"),Transform=(Rotation=(X=0,Y=0,Z=0,W=1),Translation=(X=0,Y=0,Z=0),Scale3D=(X=0.8,Y=0.8,Z=0.8))))
    Weight=1
   )
   /Script/ControlRig.RigUnit_AimItem(
    DebugSettings=(Scale=10,WorldOffset=(Rotation=(X=0,Y=0,Z=0,W=1),Translation=(X=0,Y=0,Z=0),Scale3D=(X=1,Y=1,Z=1)))
    ExecutePin=()
    Item=(Type=Bone,Name="dyn_hair_base")
    Primary=(Weight=1,Axis=(X=1,Y=0,Z=0),Target=(X=1,Y=0,Z=0),Kind=Location,Space=(Type=Bone,Name="dyn_hair_aim"))
    Secondary=(Weight=0,Axis=(X=0,Y=0,Z=-1),Target=(X=1,Y=0,Z=0),Kind=Direction,Space=(Type=Bone,Name="head"))
    Weight=1
   ) {is=struct RigVMUnitNode}
   RigVMFunction_MathVectorDistance = &RigVMFunction_MathVectorDistance(
    A=&RigUnit_GetTransform(Item=(Type=Bone,Name="dyn_hair_base"), bInitial=False).Transform.Translation
    B=&RigUnit_GetTransform(Item=(Type=Bone,Name="clavicle_r"), bInitial=False).Transform.Translation
   )
   DISPATCH_RigVMDispatch_Print = RigVMDispatchNode(Enabled=True, ExecuteContext=(), ScreenColor=(R=1,G=1,B=1,A=1), ScreenDuration=0.05, Value=@RigVMFunction_MathVectorDistance.Result) {template="DISPATCH_RigVMDispatch_Print(in Prefix,in Value,in Enabled,in ScreenDuration,in ScreenColor)"}
   n22 = &RigUnit_GetRelativeTransformForItem(Child=(Type=Bone,Name="head"), Parent=(Type=Bone,Name="spine_05"))
   &RigUnit_ModifyTransforms(
    ExecutePin=()
    ItemToModify=((Item=(Type=Bone,Name="dyn_hair_aim"),Transform=(Rotation=(X=0,Y=0,Z=0,W=1),Translation=(X=0,Y=0,Z=0),Scale3D=(X=1,Y=1,Z=1))))
    Weight=&RigVMFunction_MathFloatRemap(
     SourceMaximum=30
     SourceMinimum=-40
     TargetMinimum=0
     bClamp=true
     Value=&RigVMFunction_MathQuaternionToEuler(RotationOrder=ZYX, Value=@n22.RelativeTransform.Rotation).Result.Z
    ).Result
   )
   Remap = &RigVMFunction_MathDoubleRemap(SourceMaximum=20, SourceMinimum=28, TargetMaximum=1, TargetMinimum=0, bClamp=False, Value=@RigVMFunction_MathVectorDistance.Result)
   &RigUnit_SphericalPoseReader(
    ActiveRegionScaleFactors=(PositiveWidth=1,NegativeWidth=0.5,PositiveHeight=0.75,NegativeHeight=0.75)
    ActiveRegionSize=0.01
    Debug=(bDrawDebug=true,bDraw2D=False,bDrawLocalAxes=False,DebugScale=25,DebugSegments=20,DebugThickness=0.25)
    DriverAxis=(X=0,Y=1,Z=0)
    DriverItem=(Type=Bone,Name="head")
    ExecutePin=()
    FalloffRegionScaleFactors=(PositiveWidth=1,NegativeWidth=0.5,PositiveHeight=1,NegativeHeight=1)
    FalloffSize=0.3
    OptionalParentItem=(Type=Bone)
    RotationOffset=(X=90,Y=0,Z=-30)
   )
   &RigUnit_SphericalPoseReader(
    ActiveRegionScaleFactors=(PositiveWidth=1,NegativeWidth=1,PositiveHeight=1,NegativeHeight=1)
    ActiveRegionSize=0.01
    Debug=(bDrawDebug=true,bDraw2D=False,bDrawLocalAxes=False,DebugScale=25,DebugSegments=20,DebugThickness=0.25)
    DriverAxis=(X=-1,Y=0,Z=0)
    DriverItem=(Type=Bone,Name="upperarm_r")
    ExecutePin=()
    FalloffRegionScaleFactors=(PositiveWidth=1,NegativeWidth=1,PositiveHeight=1,NegativeHeight=1)
    FalloffSize=0.3
    OptionalParentItem=(Type=Bone)
    RotationOffset=(X=-105,Y=0,Z=25)
   )
  ]}
]
hierarchy [
 root {loc=<0 0 0>, rot=<0 0 0 1>}
 pelvis {parent=root, loc=<0 -0.5301 98.0571>, rot=<0.0127 0.707 -0.0127 -0.707>}
 spine_01 {parent=pelvis, loc=<6.1208 -0.9328 0>, rot=<0 0 0.0576 -0.9983>}
 spine_02 {parent=spine_01, loc=<6.6022 -0.0153 0>, rot=<0 0 0.0022 -1>}
 spine_03 {parent=spine_02, loc=<6.4919 -0.0105 0>, rot=<0 0 0.0015 -1>}
 spine_04 {parent=spine_03, loc=<6.9596 -0.007 0>, rot=<0 0 0.0009 -1>}
 spine_05 {parent=spine_04, loc=<6.9981 -0.0017 0>, rot=<0 0 -0.0042 -1>}
 clavicle_l {parent=spine_05, loc=<5.4831 -0.1053 -3.2946>, rot=<0.0938 -0.7388 0.0307 -0.6666>}
 upperarm_l {parent=clavicle_l, loc=<9.9486 0 0>, rot=<0.0371 -0.2992 -0.0705 -0.9509>}
 neck_01 {parent=spine_05, loc=<11.3138 0.1529 0>, rot=<0 0 -0.2003 -0.9797>}
 neck_02 {parent=neck_01, loc=<4.5627 0 0>, rot=<0 0 0.0504 -0.9987>}
 head {parent=neck_02, loc=<5.3914 0 0>, rot=<0 0 0.115 -0.9934>}
 head_offset {parent=head, loc=<8.6151 0.8524 0.1745>, rot=<0.0029 -0.7071 0.7071 -0.0029>}
 dyn_hair_base {parent=head_offset, loc=<-0.9238 -0.0388 12.5826>, rot=<0.9962 0 0.0872 0>}
 dyn_braid_base_a {parent=dyn_hair_base, loc=<-2.1325 2.9398 0.8924>, rot=<0.0023 0.0419 -0.0261 -0.9988>}
 dyn_braid_parent_a {parent=dyn_braid_base_a, loc=<0 0 0>, rot=<0 0 0 1>}
 dyn_braid_01_a {parent=dyn_braid_parent_a, loc=<15.0896 0.0274 -0.6856>, rot=<0.0523 0 0 -0.9986>}
 dyn_braid_02_a {parent=dyn_braid_01_a, loc=<18.3413 0.4477 0.2957>, rot=<-0.0341 0.0097 -0.0108 -0.9993>}
 dyn_braid_base_b {parent=dyn_hair_base, loc=<-1.1291 1.7749 -1.3462>, rot=<0 0.0175 0 -0.9998>}
 dyn_braid_parent_b {parent=dyn_braid_base_b, loc=<0 0 0>, rot=<0 0 0 1>}
 dyn_braid_01_b {parent=dyn_braid_parent_b, loc=<15.0542 0.3521 0.2426>, rot=<0 0.0175 0 -0.9998>}
 dyn_braid_02_b {parent=dyn_braid_01_b, loc=<18.9543 1.0562 0.463>, rot=<0.0014 0.007 -0.0261 -0.9996>}
 dyn_braid_base_c {parent=dyn_hair_base, loc=<-3.2741 -0.2175 -2.4254>, rot=<0 -0.0872 0 -0.9962>}
 dyn_braid_parent_c {parent=dyn_braid_base_c, loc=<0 0 0>, rot=<0 0 0 1>}
 dyn_braid_01_c {parent=dyn_braid_parent_c, loc=<17.3128 0.0006 3.4563>, rot=<0 0.1392 0 -0.9903>}
 dyn_braid_02_c {parent=dyn_braid_01_c, loc=<18.7375 0.0029 -0.2753>, rot=<0.0002 0.0005 -0.0061 -1>}
 dyn_braid_base_d {parent=dyn_hair_base, loc=<-1.2269 -0.1169 1.4516>, rot=<0.0511 -0.0436 0.0113 -0.9977>}
 dyn_braid_parent_d {parent=dyn_braid_base_d, loc=<0 0 0>, rot=<0 0 0 1>}
 dyn_braid_01_d {parent=dyn_braid_parent_d, loc=<19.6235 0.2051 2.283>, rot=<0 0.0785 0 -0.9969>}
 dyn_braid_02_d {parent=dyn_braid_01_d, loc=<13.841 0.1812 0.5193>, rot=<0 0 0 1>}
 dyn_braid_base_e {parent=dyn_hair_base, loc=<-2.5344 -2.2242 -1.0285>, rot=<0.0462 -0.0274 0.0676 -0.9963>}
 dyn_braid_parent_e {parent=dyn_braid_base_e, loc=<0 0 0>, rot=<0 0 0 1>}
 dyn_braid_01_e {parent=dyn_braid_parent_e, loc=<16.2904 1.8035 1.4197>, rot=<0.0012 0.0673 -0.0428 -0.9968>}
 dyn_braid_02_e {parent=dyn_braid_01_e, loc=<18.9902 -0.3376 0.2233>, rot=<0.0002 -0.0343 0.0067 -0.9994>}
 dyn_braid_base_f {parent=dyn_hair_base, loc=<-2.4551 -3.1369 1.2793>, rot=<-0.0052 -0.0767 0.0592 -0.9953>}
 dyn_braid_parent_f {parent=dyn_braid_base_f, loc=<0 0 0>, rot=<0 0 0 1>}
 dyn_braid_01_f {parent=dyn_braid_parent_f, loc=<15.1562 0.7579 2.5474>, rot=<-0.0028 0.1223 -0.0276 -0.9921>}
 dyn_braid_02_f {parent=dyn_braid_01_f, loc=<22.4038 -0.3322 -0.0073>, rot=<0.0003 -0.0031 -0.0045 -1>}
 dyn_hair_tie_base_r {parent=head_offset, loc=<-1.5802 -3.2462 7.198>, rot=<0.8661 -0.0121 0.1386 -0.4801>}
 dyn_hair_tie_parent_r {parent=dyn_hair_tie_base_r, loc=<0 0 0>, rot=<0 0 0 1>}
 dyn_hair_tie_base_l {parent=head_offset, loc=<1.4898 4.0893 7.7539>, rot=<0.9609 0.0407 0.1149 0.2485>}
 dyn_hair_tie_parent_l {parent=dyn_hair_tie_base_l, loc=<0 0 0>, rot=<0 0 0 1>}
 clavicle_r {parent=spine_05, loc=<5.4831 -0.1053 3.2946>, rot=<0.7388 0.0938 -0.6666 -0.0307>}
 upperarm_r {parent=clavicle_r, loc=<-9.9486 0 0>, rot=<0.0371 -0.2992 -0.0705 -0.9509>}
 dyn_hair_aim {parent=spine_05, loc=<0.6264 -7.7915 0>, rot=<-0.4795 -0.5196 0.5196 0.4795>}
]
variables [
 aim {type=bool, default=False}
 isWearingCape {type=bool, default=False}
]
