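# Physics asset for the Rays female character's face accessory part, pack-equipped variant, rigging six dynamic hair braids.
# Each braid chains a base sphere through three capsules with limited swing and twist, while the bases rotate freely relative to spine_04.
# Tight cross-constraints tie neighboring parent and first segments together and to braid d, so the braids move as a bundle.
# Spine, clavicle, hair-base and hair-aim capsules supply the surrounding body volumes.
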PhysicsAsset /BRCosmetics/Characters/Player/Female/Medium/Bodies/F_MED_Rays/Meshes/Parts/F_MED_Rays_FaceAcc_PackEquipped_Physics
bodies {
 clavicle_l [{shapeType=Capsule, length=10, radius=5, shapeName=clavicle_l_capsule, center=<11.9532 0 0>, rotation=<rot 90 0 0>}]
 clavicle_r [{shapeType=Capsule, length=10, radius=5, shapeName=clavicle_r_capsule, center=<-11.9532 0 0>, rotation=<rot 90 0 0>}]
 dyn_braid_01_a [{
   shapeType=Capsule
   length=15.2537
   radius=1.0119
   shapeName=dyn_braid_01_a_capsule
   center=<8.6637 0.0282 0.3116>
   rotation=<rot 90 180 178.3284>
  }]
 dyn_braid_01_b [{
   shapeType=Capsule
   length=17.1779
   radius=1
   shapeName=dyn_braid_01_b_capsule
   center=<8.5385 0.3168 0.3938>
   rotation=<rot 87.9791 -177.8753 179.9998>
  }]
 dyn_braid_01_c [{shapeType=Capsule, length=16.2206, radius=1, shapeName=dyn_braid_01_c_capsule, center=<8.7504 0 0>, rotation=<rot 90 0 0>}]
 dyn_braid_01_d [{
   shapeType=Capsule
   length=9.1621
   radius=1.4
   shapeName=dyn_braid_01_d_capsule
   center=<6.787 0 -0.1192>
   rotation=<rot 90 -179.3705 -179.3705>
  }]
 dyn_braid_01_e [{shapeType=Capsule, length=17.053, radius=1, shapeName=dyn_braid_01_e_capsule, center=<9.1487 0 0>, rotation=<rot 90 0 0>}]
 dyn_braid_01_f [{shapeType=Capsule, length=20.1687, radius=0.75, shapeName=dyn_braid_01_f_capsule, center=<10.0064 0 0>, rotation=<rot 90 0 0>}]
 dyn_braid_02_a [{shapeType=Capsule, length=9.1384, radius=0.946, shapeName=dyn_braid_02_a_capsule, center=<5.1022 0 0>, rotation=<rot 90 0 0>}]
 dyn_braid_02_b [{shapeType=Capsule, length=9.0083, radius=1.0781, shapeName=dyn_braid_02_b_capsule, center=<5.1178 0 0>, rotation=<rot 90 0 0>}]
 dyn_braid_02_c [{shapeType=Capsule, length=10.5106, radius=1.0614, shapeName=dyn_braid_02_c_capsule, center=<5.239 0 0>, rotation=<rot 90 0 0>}]
 dyn_braid_02_d [{shapeType=Capsule, length=8.2037, radius=1.6271, shapeName=dyn_braid_02_d_capsule, center=<6.185 0 0.6013>, rotation=<rot 90 0 0>}]
 dyn_braid_02_e [{
   shapeType=Capsule
   length=9.3525
   radius=0.8481
   shapeName=dyn_braid_02_e_capsule
   center=<5.0803 -0.1154 0.279>
   rotation=<rot 84.5483 -180 -179.9999>
  }]
 dyn_braid_02_f [{shapeType=Capsule, length=5.8861, radius=0.851, shapeName=dyn_braid_02_f_capsule, center=<2.8709 0 0>, rotation=<rot 90 0 0>}]
 dyn_braid_base_a [{shapeType=Sphere, radius=0.5, shapeName=dyn_braid_base_a_sphere}]
 dyn_braid_base_b [{shapeType=Sphere, radius=0.5, shapeName=dyn_braid_base_b_sphere}]
 dyn_braid_base_c [{shapeType=Sphere, radius=0.5, shapeName=dyn_braid_base_c_sphere}]
 dyn_braid_base_d [{shapeType=Sphere, radius=0.5, shapeName=dyn_braid_base_d_sphere}]
 dyn_braid_base_e [{shapeType=Sphere, radius=0.5, shapeName=dyn_braid_base_e_sphere}]
 dyn_braid_base_f [{shapeType=Sphere, radius=0.5, shapeName=dyn_braid_base_f_sphere}]
 dyn_braid_parent_a [{
   shapeType=Capsule
   length=13.9066
   radius=1
   shapeName=dyn_braid_parent_a_capsule
   center=<7.4638 -0.166 0>
   rotation=<rot 89.3938 0.0009 0.0009>
  }]
 dyn_braid_parent_b [{
   shapeType=Capsule
   length=15.5726
   radius=1
   shapeName=dyn_braid_parent_b_capsule
   center=<8.482 -0.1619 0.4965>
   rotation=<rot 87.8266 180 177.2393>
  }]
 dyn_braid_parent_c [{
   shapeType=Capsule
   length=15.8514
   radius=0.75
   shapeName=dyn_braid_parent_c_capsule
   center=<8.2371 0 1.6304>
   rotation=<rot 79.3496 180 180>
  }]
 dyn_braid_parent_d [{
   shapeType=Capsule
   length=15.5142
   radius=1.4
   shapeName=dyn_braid_parent_d_capsule
   center=<8.9351 0 0.7461>
   rotation=<rot 81.4905 -180 -180>
  }]
 dyn_braid_parent_e [{
   shapeType=Capsule
   length=17.3874
   radius=0.75
   shapeName=dyn_braid_parent_e_capsule
   center=<9.1459 1.0664 0.4254>
   rotation=<rot 84.9342 -175.396 -179.9999>
  }]
 dyn_braid_parent_f [{
   shapeType=Capsule
   length=14.2108
   radius=1.5
   shapeName=dyn_braid_parent_f_capsule
   center=<10 -0.053 0.6636>
   rotation=<rot 76.9431 -179.9998 176.375>
  }]
 dyn_hair_aim [{shapeType=Capsule, length=25.9019, radius=15, shapeName=dyn_hair_aim_capsule, center=<-20.8224 0 -0.97>, rotation=<rot 0 0 90>}]
 dyn_hair_base [{shapeType=Capsule, length=2.8396, radius=2.3956, shapeName=dyn_hair_base_capsule, center=<3.0523 0 0>, rotation=<rot 90 0 0>}]
 spine_01 [{shapeType=Capsule, length=9.0908, radius=17.6788, shapeName=spine_01_capsule, center=<0 8.5651 0>, rotation=<rot 0 -90 0>}]
 spine_04 [
  {shapeType=Capsule, length=28.3686, radius=15, shapeName=spine_04_capsule, center=<-6.385 -27.79 0>}
  {shapeType=Capsule, length=10.8635, radius=25.642, shapeName=spine_04_capsule, center=<2.7305 17 0>, rotation=<rot 90 180 177.0001>}
 ]
}
constraints [
 {bone1Name=dyn_braid_02_f, bone2Name=dyn_braid_01_f, swing1LimitDegrees=30, swing2LimitDegrees=15, twistLimitDegrees=8}
 {bone1Name=dyn_braid_01_f, bone2Name=dyn_braid_parent_f, swing1LimitDegrees=15, swing2LimitDegrees=25, twistLimitDegrees=8}
 {bone1Name=dyn_braid_parent_f, bone2Name=dyn_braid_base_f, swing1LimitDegrees=10, swing2LimitDegrees=8, twistLimitDegrees=5}
 {bone1Name=dyn_braid_01_e, bone2Name=dyn_braid_parent_e, swing1LimitDegrees=15, swing2LimitDegrees=25, twistLimitDegrees=8}
 {bone1Name=dyn_braid_parent_e, bone2Name=dyn_braid_base_e, swing1LimitDegrees=10, swing2LimitDegrees=8, twistLimitDegrees=5}
 {bone1Name=dyn_braid_02_d, bone2Name=dyn_braid_01_d, swing1LimitDegrees=10, swing2LimitDegrees=5, twistLimitDegrees=2}
 {bone1Name=dyn_braid_01_d, bone2Name=dyn_braid_parent_d, swing1LimitDegrees=5, swing2LimitDegrees=15, twistLimitDegrees=3}
 {bone1Name=dyn_braid_parent_d, bone2Name=dyn_braid_base_d, swing1LimitDegrees=10, swing2LimitDegrees=15, twistLimitDegrees=5}
 {bone1Name=dyn_braid_02_c, bone2Name=dyn_braid_01_c, swing1LimitDegrees=30, swing2LimitDegrees=15, twistLimitDegrees=8}
 {bone1Name=dyn_braid_01_c, bone2Name=dyn_braid_parent_c, swing1LimitDegrees=15, swing2LimitDegrees=25, twistLimitDegrees=8}
 {bone1Name=dyn_braid_parent_c, bone2Name=dyn_braid_base_c, swing1LimitDegrees=10, swing2LimitDegrees=8, twistLimitDegrees=5}
 {bone1Name=dyn_braid_02_b, bone2Name=dyn_braid_01_b, swing1LimitDegrees=30, swing2LimitDegrees=15, twistLimitDegrees=8}
 {bone1Name=dyn_braid_01_b, bone2Name=dyn_braid_parent_b, swing1LimitDegrees=15, swing2LimitDegrees=25, twistLimitDegrees=8}
 {bone1Name=dyn_braid_parent_b, bone2Name=dyn_braid_base_b, swing1LimitDegrees=10, swing2LimitDegrees=8, twistLimitDegrees=5}
 {bone1Name=dyn_braid_02_a, bone2Name=dyn_braid_01_a, swing1LimitDegrees=15, swing2LimitDegrees=15, twistLimitDegrees=8}
 {bone1Name=dyn_braid_01_a, bone2Name=dyn_braid_parent_a, swing1LimitDegrees=10, swing2LimitDegrees=25, twistLimitDegrees=8}
 {bone1Name=dyn_braid_parent_a, bone2Name=dyn_braid_base_a, swing1LimitDegrees=10, swing2LimitDegrees=8, twistLimitDegrees=5}
 {bone1Name=dyn_hair_base, bone2Name=spine_04, swing1LimitDegrees=10, swing2LimitDegrees=8, twistLimitDegrees=5}
 {bone1Name=dyn_braid_02_e, bone2Name=dyn_braid_01_e, swing1LimitDegrees=30, swing2LimitDegrees=15, twistLimitDegrees=8}
 {
  bone1Name=clavicle_l
  bone2Name=spine_04
  swing1LimitDegrees=45
  swing1Motion=Free
  swing2LimitDegrees=45
  swing2Motion=Free
  twistLimitDegrees=45
  twistMotion=Free
 }
 {
  bone1Name=clavicle_r
  bone2Name=spine_04
  swing1LimitDegrees=45
  swing1Motion=Free
  swing2LimitDegrees=45
  swing2Motion=Free
  twistLimitDegrees=45
  twistMotion=Free
 }
 {
  bone1Name=dyn_braid_base_a
  bone2Name=spine_04
  swing1LimitDegrees=45
  swing1Motion=Free
  swing2LimitDegrees=45
  swing2Motion=Free
  twistLimitDegrees=45
  twistMotion=Free
 }
 {
  bone1Name=dyn_braid_base_b
  bone2Name=spine_04
  swing1LimitDegrees=45
  swing1Motion=Free
  swing2LimitDegrees=45
  swing2Motion=Free
  twistLimitDegrees=45
  twistMotion=Free
 }
 {
  bone1Name=dyn_braid_base_c
  bone2Name=spine_04
  swing1LimitDegrees=45
  swing1Motion=Free
  swing2LimitDegrees=45
  swing2Motion=Free
  twistLimitDegrees=45
  twistMotion=Free
 }
 {
  bone1Name=dyn_braid_base_d
  bone2Name=spine_04
  swing1LimitDegrees=45
  swing1Motion=Free
  swing2LimitDegrees=45
  swing2Motion=Free
  twistLimitDegrees=45
  twistMotion=Free
 }
 {
  bone1Name=dyn_braid_base_e
  bone2Name=spine_04
  swing1LimitDegrees=45
  swing1Motion=Free
  swing2LimitDegrees=45
  swing2Motion=Free
  twistLimitDegrees=45
  twistMotion=Free
 }
 {
  bone1Name=dyn_braid_base_f
  bone2Name=spine_04
  swing1LimitDegrees=45
  swing1Motion=Free
  swing2LimitDegrees=45
  swing2Motion=Free
  twistLimitDegrees=45
  twistMotion=Free
 }
 {
  bone1Name=dyn_hair_aim
  bone2Name=spine_04
  swing1LimitDegrees=45
  swing1Motion=Free
  swing2LimitDegrees=45
  swing2Motion=Free
  twistLimitDegrees=45
  twistMotion=Free
 }
 {bone1Name=dyn_braid_parent_a, bone2Name=dyn_braid_parent_d, swing1LimitDegrees=5, swing2LimitDegrees=5, twistLimitDegrees=1}
 {bone1Name=dyn_braid_parent_b, bone2Name=dyn_braid_parent_d, swing1LimitDegrees=5, swing2LimitDegrees=5, twistLimitDegrees=1}
 {bone1Name=dyn_braid_parent_c, bone2Name=dyn_braid_parent_d, swing1LimitDegrees=5, swing2LimitDegrees=5, twistLimitDegrees=1}
 {bone1Name=dyn_braid_parent_e, bone2Name=dyn_braid_parent_d, swing1LimitDegrees=5, swing2LimitDegrees=5, twistLimitDegrees=1}
 {bone1Name=dyn_braid_parent_f, bone2Name=dyn_braid_parent_d, swing1LimitDegrees=5, swing2LimitDegrees=5, twistLimitDegrees=1}
 {bone1Name=dyn_braid_parent_a, bone2Name=dyn_braid_parent_b, swing1LimitDegrees=5, swing2LimitDegrees=5, twistLimitDegrees=1}
 {bone1Name=dyn_braid_parent_b, bone2Name=dyn_braid_parent_c, swing1LimitDegrees=5, swing2LimitDegrees=5, twistLimitDegrees=1}
 {bone1Name=dyn_braid_parent_c, bone2Name=dyn_braid_parent_e, swing1LimitDegrees=5, swing2LimitDegrees=5, twistLimitDegrees=1}
 {bone1Name=dyn_braid_parent_e, bone2Name=dyn_braid_parent_f, swing1LimitDegrees=5, swing2LimitDegrees=5, twistLimitDegrees=1}
 {
  bone1Name=spine_04
  bone2Name=spine_01
  swing1LimitDegrees=45
  swing1Motion=Free
  swing2LimitDegrees=45
  swing2Motion=Free
  twistLimitDegrees=45
  twistMotion=Free
 }
 {bone1Name=dyn_braid_01_a, bone2Name=dyn_braid_01_b, swing1LimitDegrees=5, swing2LimitDegrees=10, twistLimitDegrees=1}
 {bone1Name=dyn_braid_01_b, bone2Name=dyn_braid_01_c, swing1LimitDegrees=5, swing2LimitDegrees=10, twistLimitDegrees=1}
 {bone1Name=dyn_braid_01_c, bone2Name=dyn_braid_01_e, swing1LimitDegrees=5, swing2LimitDegrees=10, twistLimitDegrees=1}
 {bone1Name=dyn_braid_01_e, bone2Name=dyn_braid_01_f, swing1LimitDegrees=5, swing2LimitDegrees=10, twistLimitDegrees=1}
 {bone1Name=dyn_braid_01_e, bone2Name=dyn_braid_01_d, swing1LimitDegrees=5, swing2LimitDegrees=5, twistLimitDegrees=1}
 {bone1Name=dyn_braid_01_f, bone2Name=dyn_braid_01_d, swing1LimitDegrees=5, swing2LimitDegrees=5, twistLimitDegrees=1}
 {bone1Name=dyn_braid_01_c, bone2Name=dyn_braid_01_d, swing1LimitDegrees=5, swing2LimitDegrees=5, twistLimitDegrees=1}
 {bone1Name=dyn_braid_01_b, bone2Name=dyn_braid_01_d, swing1LimitDegrees=5, swing2LimitDegrees=5, twistLimitDegrees=1}
 {bone1Name=dyn_braid_01_a, bone2Name=dyn_braid_01_d, swing1LimitDegrees=5, swing2LimitDegrees=5, twistLimitDegrees=1}
]
