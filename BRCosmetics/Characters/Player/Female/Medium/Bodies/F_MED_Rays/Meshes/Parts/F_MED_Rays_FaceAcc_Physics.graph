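# Physics asset for the F_MED_Rays FaceAcc mesh that sets up six dynamic braids with torso and clavicle bodies.
# Each braid chains base, parent, 01 and 02 bones under tight twist limits, with bases angularly free against spine_04.
# Cross-constraints tie neighboring braids at parent and 01 levels, with braid d as a hub, so they move as a bundle.
# Braid d's lower segments each use three capsules, and dyn_hair_base has no constraint.
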PhysicsAsset /BRCosmetics/Characters/Player/Female/Medium/Bodies/F_MED_Rays/Meshes/Parts/F_MED_Rays_FaceAcc_Physics
bodies {
 clavicle_l [{
   shapeType=Capsule
   length=19.0579
   radius=8.9421
   shapeName=clavicle_l_capsule
   center=<18.6889 0 -6.1105>
   rotation=<rot -25.665 180 166.0837>
  }]
 clavicle_r [{shapeType=Capsule, length=19.0579, radius=8.9421, shapeName=clavicle_r_capsule, center=<-18.7 0 6>, rotation=<rot -25.665 180 166.0837>}]
 dyn_braid_01_a [{
   shapeType=Capsule
   length=16.0452
   radius=0.9747
   shapeName=dyn_braid_01_a_capsule
   center=<8.6551 0.2216 0.1869>
   rotation=<rot 89.5102 -178.1955 179.9955>
  }]
 dyn_braid_01_b [{
   shapeType=Capsule
   length=17.5098
   radius=1.1424
   shapeName=dyn_braid_01_b_capsule
   center=<8.5385 0.3168 0.3938>
   rotation=<rot 87.9791 -177.8753 179.9998>
  }]
 dyn_braid_01_c [{shapeType=Capsule, length=16.4502, radius=0.9783, shapeName=dyn_braid_01_c_capsule, center=<8.7504 0 0>, rotation=<rot 90 0 0>}]
 dyn_braid_01_d [
  {
   shapeType=Capsule
   length=14.3901
   radius=1.1112
   shapeName=dyn_braid_01_d_capsule
   center=<6.787 0.1385 -0.57>
   rotation=<rot 90 -179.9209 -179.9208>
  }
  {
   shapeType=Capsule
   length=14.3901
   radius=1.1112
   shapeName=dyn_braid_01_d_capsule
   center=<6.787 1.4162 2.0783>
   rotation=<rot 90 -179.3706 178.399>
  }
  {
   shapeType=Capsule
   length=14.3901
   radius=1.1112
   shapeName=dyn_braid_01_d_capsule
   center=<6.787 -1.6863 1.1516>
   rotation=<rot 90 -10.025 -10.025>
  }
 ]
 dyn_braid_01_e [{shapeType=Capsule, length=17.2826, radius=1.1307, shapeName=dyn_braid_01_e_capsule, center=<9.1487 0 0>, rotation=<rot 90 0 0>}]
 dyn_braid_01_f [{shapeType=Capsule, length=20.6483, radius=1.0168, shapeName=dyn_braid_01_f_capsule, center=<10.0064 0 0>, rotation=<rot 90 0 0>}]
 dyn_braid_02_a [{shapeType=Capsule, length=9.2036, radius=0.8809, shapeName=dyn_braid_02_a_capsule, center=<5.1022 0 0>, rotation=<rot 90 0 0>}]
 dyn_braid_02_b [{shapeType=Capsule, length=9.1174, radius=0.969, shapeName=dyn_braid_02_b_capsule, center=<5.1178 0 0>, rotation=<rot 90 0 0>}]
 dyn_braid_02_c [{shapeType=Capsule, length=10.5429, radius=1.0291, shapeName=dyn_braid_02_c_capsule, center=<5.239 0 0>, rotation=<rot 90 0 0>}]
 dyn_braid_02_d [
  {shapeType=Capsule, length=10.6072, radius=0.7841, shapeName=dyn_braid_02_d_capsule, center=<6.185 -0.0164 -0.7426>, rotation=<rot 90 0 0>}
  {shapeType=Capsule, length=10.6072, radius=0.7841, shapeName=dyn_braid_02_d_capsule, center=<6.185 1.5557 1.6728>, rotation=<rot 90 0 0>}
  {shapeType=Capsule, length=10.6072, radius=0.7841, shapeName=dyn_braid_02_d_capsule, center=<6.185 -1.8468 1.0584>, rotation=<rot 90 0 0>}
 ]
 dyn_braid_02_e [{
   shapeType=Capsule
   length=9.3074
   radius=0.8933
   shapeName=dyn_braid_02_e_capsule
   center=<5.0803 -0.1154 0.279>
   rotation=<rot 84.5483 -180 -179.9999>
  }]
 dyn_braid_02_f [{shapeType=Capsule, length=5.7452, radius=0.9919, shapeName=dyn_braid_02_f_capsule, center=<2.8709 0 0>, rotation=<rot 90 0 0>}]
 dyn_braid_base_a [{shapeType=Sphere, radius=0.5, shapeName=dyn_braid_base_a_sphere}]
 dyn_braid_base_b [{shapeType=Sphere, radius=0.5, shapeName=dyn_braid_base_b_sphere}]
 dyn_braid_base_c [{shapeType=Sphere, radius=0.5, shapeName=dyn_braid_base_c_sphere}]
 dyn_braid_base_d [{shapeType=Sphere, radius=0.5, shapeName=dyn_braid_base_d_sphere}]
 dyn_braid_base_e [{shapeType=Sphere, radius=0.5, shapeName=dyn_braid_base_e_sphere}]
 dyn_braid_base_f [{shapeType=Sphere, radius=0.5, shapeName=dyn_braid_base_f_sphere}]
 dyn_braid_parent_a [{
   shapeType=Capsule
   length=14.4662
   radius=0.7903
   shapeName=dyn_braid_parent_a_capsule
   center=<7.4638 -0.166 0>
   rotation=<rot 89.3938 0.0009 0.0009>
  }]
 dyn_braid_parent_b [{
   shapeType=Capsule
   length=16.2128
   radius=0.7235
   shapeName=dyn_braid_parent_b_capsule
   center=<8.463 0.2314 0.4957>
   rotation=<rot 87.8266 180 177.2393>
  }]
 dyn_braid_parent_c [{
   shapeType=Capsule
   length=15.7278
   radius=1.1326
   shapeName=dyn_braid_parent_c_capsule
   center=<8.2371 0 1.6304>
   rotation=<rot 79.3496 180 180>
  }]
 dyn_braid_parent_d [{
   shapeType=Capsule
   length=14.4525
   radius=2.7685
   shapeName=dyn_braid_parent_d_capsule
   center=<8.9351 0 0.7461>
   rotation=<rot 81.4905 -180 -180>
  }]
 dyn_braid_parent_e [{
   shapeType=Capsule
   length=17.0878
   radius=1.1376
   shapeName=dyn_braid_parent_e_capsule
   center=<9.1459 1.0664 0.4254>
   rotation=<rot 84.9342 -175.396 -179.9999>
  }]
 dyn_braid_parent_f [{
   shapeType=Capsule
   length=14.6612
   radius=1.1316
   shapeName=dyn_braid_parent_f_capsule
   center=<8.0716 -0.053 0.6636>
   rotation=<rot 76.9431 -179.9998 176.375>
  }]
 dyn_hair_base [{shapeType=Capsule, length=2.8396, radius=2.3956, shapeName=dyn_hair_base_capsule, center=<3.0523 0 0>, rotation=<rot 90 0 0>}]
 spine_01 [{shapeType=Capsule, length=9.0908, radius=17.6788, shapeName=spine_01_capsule, center=<0 6.4116 0>, rotation=<rot 0 -90 0>}]
 spine_04 [{shapeType=Capsule, length=13.3635, radius=23.142, shapeName=spine_04_capsule, center=<2.7495 13.2127 0>, rotation=<rot 90 180 177.0001>}]
}
constraints [
 {bone1Name=dyn_braid_02_f, bone2Name=dyn_braid_01_f, swing1LimitDegrees=15, swing2LimitDegrees=40, twistLimitDegrees=3}
 {bone1Name=dyn_braid_01_f, bone2Name=dyn_braid_parent_f, swing1LimitDegrees=15, swing2LimitDegrees=40, twistLimitDegrees=3}
 {bone1Name=dyn_braid_parent_f, bone2Name=dyn_braid_base_f, swing1LimitDegrees=15, swing2LimitDegrees=40, twistLimitDegrees=3}
 {bone1Name=dyn_braid_01_e, bone2Name=dyn_braid_parent_e, swing1LimitDegrees=15, swing2LimitDegrees=40, twistLimitDegrees=3}
 {bone1Name=dyn_braid_parent_e, bone2Name=dyn_braid_base_e, swing1LimitDegrees=20, swing2LimitDegrees=40, twistLimitDegrees=3}
 {bone1Name=dyn_braid_02_d, bone2Name=dyn_braid_01_d, swing1LimitDegrees=30, swing2LimitDegrees=40, twistLimitDegrees=3}
 {bone1Name=dyn_braid_01_d, bone2Name=dyn_braid_parent_d, swing1LimitDegrees=20, swing2LimitDegrees=40, twistLimitDegrees=3}
 {bone1Name=dyn_braid_parent_d, bone2Name=dyn_braid_base_d, swing1LimitDegrees=20, swing2LimitDegrees=40, twistLimitDegrees=3}
 {bone1Name=dyn_braid_02_c, bone2Name=dyn_braid_01_c, swing1LimitDegrees=15, swing2LimitDegrees=40, twistLimitDegrees=3}
 {bone1Name=dyn_braid_01_c, bone2Name=dyn_braid_parent_c, swing1LimitDegrees=20, swing2LimitDegrees=40, twistLimitDegrees=3}
 {bone1Name=dyn_braid_parent_c, bone2Name=dyn_braid_base_c, swing1LimitDegrees=20, swing2LimitDegrees=40, twistLimitDegrees=3}
 {bone1Name=dyn_braid_02_b, bone2Name=dyn_braid_01_b, swing1LimitDegrees=20, swing2LimitDegrees=40, twistLimitDegrees=3}
 {bone1Name=dyn_braid_01_b, bone2Name=dyn_braid_parent_b, swing1LimitDegrees=20, swing2LimitDegrees=40, twistLimitDegrees=3}
 {bone1Name=dyn_braid_parent_b, bone2Name=dyn_braid_base_b, swing1LimitDegrees=10, swing2LimitDegrees=40, twistLimitDegrees=3}
 {bone1Name=dyn_braid_02_a, bone2Name=dyn_braid_01_a, swing1LimitDegrees=15, swing2LimitDegrees=40, twistLimitDegrees=3}
 {bone1Name=dyn_braid_01_a, bone2Name=dyn_braid_parent_a, swing1LimitDegrees=15, swing2LimitDegrees=40, twistLimitDegrees=1}
 {bone1Name=dyn_braid_parent_a, bone2Name=dyn_braid_base_a, swing1LimitDegrees=20, swing2LimitDegrees=40, twistLimitDegrees=1}
 {bone1Name=dyn_braid_02_e, bone2Name=dyn_braid_01_e, swing1LimitDegrees=20, swing2LimitDegrees=40, twistLimitDegrees=3}
 {
  bone1Name=dyn_braid_base_a
  bone2Name=spine_04
  swing1LimitDegrees=45
  swing1Motion=Free
  swing2LimitDegrees=45
  swing2Motion=Free
  twistLimitDegrees=45
  twistMotion=Free
 }
 {
  bone1Name=dyn_braid_base_b
  bone2Name=spine_04
  swing1LimitDegrees=45
  swing1Motion=Free
  swing2LimitDegrees=45
  swing2Motion=Free
  twistLimitDegrees=45
  twistMotion=Free
 }
 {
  bone1Name=dyn_braid_base_c
  bone2Name=spine_04
  swing1LimitDegrees=45
  swing1Motion=Free
  swing2LimitDegrees=45
  swing2Motion=Free
  twistLimitDegrees=45
  twistMotion=Free
 }
 {
  bone1Name=dyn_braid_base_d
  bone2Name=spine_04
  swing1LimitDegrees=45
  swing1Motion=Free
  swing2LimitDegrees=45
  swing2Motion=Free
  twistLimitDegrees=45
  twistMotion=Free
 }
 {
  bone1Name=dyn_braid_base_e
  bone2Name=spine_04
  swing1LimitDegrees=45
  swing1Motion=Free
  swing2LimitDegrees=45
  swing2Motion=Free
  twistLimitDegrees=45
  twistMotion=Free
 }
 {
  bone1Name=dyn_braid_base_f
  bone2Name=spine_04
  swing1LimitDegrees=45
  swing1Motion=Free
  swing2LimitDegrees=45
  swing2Motion=Free
  twistLimitDegrees=45
  twistMotion=Free
 }
 {
  bone1Name=spine_04
  bone2Name=spine_01
  swing1LimitDegrees=45
  swing1Motion=Free
  swing2LimitDegrees=45
  swing2Motion=Free
  twistLimitDegrees=45
  twistMotion=Free
 }
 {bone1Name=dyn_braid_parent_c, bone2Name=dyn_braid_parent_b, swing1LimitDegrees=15, swing2LimitDegrees=15, twistLimitDegrees=5}
 {bone1Name=dyn_braid_parent_b, bone2Name=dyn_braid_parent_a, swing1LimitDegrees=15, swing2LimitDegrees=15, twistLimitDegrees=5}
 {bone1Name=dyn_braid_parent_e, bone2Name=dyn_braid_parent_c, swing1LimitDegrees=15, swing2LimitDegrees=15, twistLimitDegrees=5}
 {bone1Name=dyn_braid_parent_f, bone2Name=dyn_braid_parent_e, swing1LimitDegrees=15, swing2LimitDegrees=15, twistLimitDegrees=5}
 {bone1Name=dyn_braid_parent_a, bone2Name=dyn_braid_parent_d, swing1LimitDegrees=15, swing2LimitDegrees=15, twistLimitDegrees=5}
 {bone1Name=dyn_braid_parent_b, bone2Name=dyn_braid_parent_d, swing1LimitDegrees=15, swing2LimitDegrees=15, twistLimitDegrees=5}
 {bone1Name=dyn_braid_parent_c, bone2Name=dyn_braid_parent_d, swing1LimitDegrees=15, swing2LimitDegrees=15, twistLimitDegrees=5}
 {bone1Name=dyn_braid_parent_e, bone2Name=dyn_braid_parent_d, swing1LimitDegrees=15, swing2LimitDegrees=15, twistLimitDegrees=5}
 {bone1Name=dyn_braid_parent_f, bone2Name=dyn_braid_parent_d, swing1LimitDegrees=15, swing2LimitDegrees=15, twistLimitDegrees=5}
 {
  bone1Name=clavicle_l
  bone2Name=spine_04
  swing1LimitDegrees=45
  swing1Motion=Free
  swing2LimitDegrees=45
  swing2Motion=Free
  twistLimitDegrees=45
  twistMotion=Free
 }
 {
  bone1Name=clavicle_r
  bone2Name=spine_04
  swing1LimitDegrees=45
  swing1Motion=Free
  swing2LimitDegrees=45
  swing2Motion=Free
  twistLimitDegrees=45
  twistMotion=Free
 }
 {bone1Name=dyn_braid_01_a, bone2Name=dyn_braid_01_b, swing1LimitDegrees=20, swing2LimitDegrees=20, twistLimitDegrees=3}
 {bone1Name=dyn_braid_01_b, bone2Name=dyn_braid_01_c, swing1LimitDegrees=20, swing2LimitDegrees=20, twistLimitDegrees=3}
 {bone1Name=dyn_braid_01_c, bone2Name=dyn_braid_01_e, swing1LimitDegrees=20, swing2LimitDegrees=20, twistLimitDegrees=3}
 {bone1Name=dyn_braid_01_e, bone2Name=dyn_braid_01_f, swing1LimitDegrees=20, swing2LimitDegrees=20, twistLimitDegrees=3}
 {bone1Name=dyn_braid_01_a, bone2Name=dyn_braid_01_d, swing1LimitDegrees=20, swing2LimitDegrees=20, twistLimitDegrees=3}
 {bone1Name=dyn_braid_01_b, bone2Name=dyn_braid_01_d, swing1LimitDegrees=20, swing2LimitDegrees=20, twistLimitDegrees=3}
 {bone1Name=dyn_braid_01_c, bone2Name=dyn_braid_01_d, swing1LimitDegrees=20, swing2LimitDegrees=20, twistLimitDegrees=3}
 {bone1Name=dyn_braid_01_e, bone2Name=dyn_braid_01_d, swing1LimitDegrees=20, swing2LimitDegrees=20, twistLimitDegrees=3}
 {bone1Name=dyn_braid_01_f, bone2Name=dyn_braid_01_d, swing1LimitDegrees=20, swing2LimitDegrees=20, twistLimitDegrees=3}
]
