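# Post-process style AnimBlueprint for the Rays character's FaceAcc part, animating braided hair and hair ties to follow the body and avoid backpacks and capes.
# It caches copied body poses with AnimDynamics ties, then selects trail-plus-RigidBody simulation (no pack, cape) or constrained, offset braids (short pack).
# Events derive backbling flags, movement-driven force, gravity, skydive and hair-base alphas.
# The Control Rig and PackEquipped physics chain never reaches the output.
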AnimBlueprint /BRCosmetics/Characters/Player/Female/Medium/Bodies/F_MED_Rays/Meshes/Parts/F_MED_Rays_FaceAcc_AnimBP
target_skeleton=&/F_MED_Rays_FaceAcc_Skeleton
graphs [
 AnimGraph {nodes [
   n1 = Look At - Bone: dyn_hair_base(Alpha=1) {
    alphaBoolBlend {blendOption=Linear}
    boneToModify {boneName=dyn_hair_base}
    lookAtLocation=<9.5803 -17.1412 0.1513>
    lookAtTarget {boneReference {boneName=pelvis}}
    lookAt_Axis {axis=<1 0>}
    lookUp_Axis {axis {x=0, z=-1}}
   }
   n2 = Control Rig!(Source=Component To Local(ComponentPose=@n1), Alpha, aim:bool=false) -> (Pose) {
    alphaBoolBlend {blendOption=Linear}
    controlRigAssetReference {blueprintRigClass=&/F_MED_Rays_FaceAcc_CtrlRig.F_MED_Rays_FaceAcc_CtrlRig_C}
   }
   n3 = "Transform (Modify) Bone - Bone: dyn_braid_parent_d"(ComponentPose=Local To Component(LocalPose=@n2), Alpha=1, Translation="0,0,0", Rotation="0,0,0", Scale="1,1,0.5") {
    rotationMode=BMM_Additive
    rotationSpace=BCS_BoneSpace
    scaleMode=BMM_Additive
    scaleSpace=BCS_BoneSpace
    translationMode=BMM_Additive
    translationSpace=BCS_BoneSpace
    alphaBoolBlend {blendOption=Linear}
    boneToModify {boneName=dyn_braid_parent_d}
    scale {z=0.5}
   }
   n4 = "Transform (Modify) Bone - Bone: dyn_hair_base"(ComponentPose=@n3, Alpha=Get Skydive() -> (Skydive:double), Translation="0,0,0", Rotation="-50,0.000,-0.000", Scale="1,1,1") {
    rotationMode=BMM_Additive
    rotationSpace=BCS_BoneSpace
    translationMode=BMM_Additive
    translationSpace=BCS_BoneSpace
    alphaBoolBlend {blendOption=Linear}
    boneToModify {boneName=dyn_hair_base}
    rotation {pitch=-50}
   }
   n5 = RigidBody(ComponentPose=@n4, ExternalForce=Get Velocity() -> (Velocity:Vector)) {
    bClampLinearTranslationLimitToRefPose=true
    bOverrideWorldGravity=true
    overridePhysicsAsset=&/F_MED_Rays_FaceAcc_PackEquipped_Physics
    simulationSpace=BaseBoneSpace
    alphaBoolBlend {blendOption=Linear}
    baseBoneRef {boneName=neck_02}
    componentAppliedLinearAccClamp=<6000 6000 6000>
    componentLinearAccScale=<0.1 0.1 0.1>
    componentLinearVelScale=<0.1 0.1 0.1>
    overrideWorldGravity {z=-980}
    simSpaceSettings {
     maxAngularAcceleration=100
     maxAngularVelocity=3
     maxLinearAcceleration=400
     maxLinearVelocity=250
     velocityScaleZ=0
     worldAlpha=1
     externalLinearDragV=<0.1 0.1>
    }
   }
   "Transform (Modify) Bone - Bone: dyn_hair_base"(ComponentPose=@n5, Alpha=1, Translation="0.000,0.000,0.099", Rotation="0,0,0", Scale="0.9,1,0.826") {
    rotationMode=BMM_Additive
    rotationSpace=BCS_BoneSpace
    scaleMode=BMM_Additive
    scaleSpace=BCS_BoneSpace
    translationMode=BMM_Additive
    translationSpace=BCS_BoneSpace
    alphaBoolBlend {blendOption=Linear}
    boneToModify {boneName=dyn_hair_base}
    scale {x=0.9, z=0.826}
    translation {z=0.0992}
   }
   n6 = "Transform (Modify) Bone - Bone: head_offset"!(
    ComponentPose=Local To Component(LocalPose=Use cached pose 'BackpackCache'())
    bAlphaBoolEnabled:bool=Get bIsSkyDiving()
    Translation="0,0,0"
    Rotation="30,0,0"
    Scale="1,1,1"
   ) -> (Pose) {
    alphaInputType=Bool
    rotationMode=BMM_Additive
    rotationSpace=BCS_BoneSpace
    scaleMode=BMM_Additive
    scaleSpace=BCS_BoneSpace
    translationMode=BMM_Additive
    translationSpace=BCS_BoneSpace
    alphaBoolBlend {blendOption=Linear}
    boneToModify {boneName=head_offset}
    rotation {pitch=30}
   }
   n7 = "Transform (Modify) Bone - Bone: head_offset"!(ComponentPose=@n6, Alpha:float=1, Translation="0,0,0", Rotation="10,0,0", Scale="1,1,1") -> (Pose) {
    rotationMode=BMM_Additive
    rotationSpace=BCS_BoneSpace
    scaleMode=BMM_Additive
    scaleSpace=BCS_BoneSpace
    translationMode=BMM_Additive
    translationSpace=BCS_BoneSpace
    alphaBoolBlend {blendOption=Linear}
    boneToModify {boneName=head_offset}
    rotation {pitch=10}
   }
   n8 = Trail controller - Bone: dyn_braid_01_a(ComponentPose=@n7, Alpha=0.3) {
    bLimitRotation=true
    rotationLimits [{limitMax=<1 1 1>, limitMin=<-1 -1 -1>}, {limitMax=<15 15 15>, limitMin=<-15 -15 -15>}]
    rotationOffsets [<0 0 0>, <0 0 0>]
    trailBone {boneName=dyn_braid_01_a}
   }
   n9 = Trail controller - Bone: dyn_braid_01_b(ComponentPose=@n8, Alpha=0.3) {
    bLimitRotation=true
    rotationLimits [{limitMax=<1 1 1>, limitMin=<-1 -1 -1>}, {limitMax=<15 15 15>, limitMin=<-15 -15 -15>}]
    rotationOffsets [<0 0 0>, <0 0 0>]
    trailBone {boneName=dyn_braid_01_b}
   }
   n10 = Trail controller - Bone: dyn_braid_01_c(ComponentPose=@n9, Alpha=0.3) {
    bLimitRotation=true
    rotationLimits [{limitMax=<1 1 1>, limitMin=<-1 -1 -1>}, {limitMax=<15 15 15>, limitMin=<-15 -15 -15>}]
    rotationOffsets [<0 0 0>, <0 0 0>]
    trailBone {boneName=dyn_braid_01_c}
   }
   n11 = Trail controller - Bone: dyn_braid_01_d(ComponentPose=@n10, Alpha=0.3) {
    bLimitRotation=true
    rotationLimits [{limitMax=<1 1 1>, limitMin=<-1 -1 -1>}, {limitMax=<15 15 15>, limitMin=<-15 -15 -15>}]
    rotationOffsets [<0 0 0>, <0 0 0>]
    trailBone {boneName=dyn_braid_01_d}
   }
   n12 = Trail controller - Bone: dyn_braid_01_e(ComponentPose=@n11, Alpha=0.3) {
    bLimitRotation=true
    rotationLimits [{limitMax=<1 1 1>, limitMin=<-1 -1 -1>}, {limitMax=<15 15 15>, limitMin=<-15 -15 -15>}]
    rotationOffsets [<0 0 0>, <0 0 0>]
    trailBone {boneName=dyn_braid_01_e}
   }
   n13 = Trail controller - Bone: dyn_braid_01_f(ComponentPose=@n12, Alpha=0.3) {
    bLimitRotation=true
    rotationLimits [{limitMax=<1 1 1>, limitMin=<-1 -1 -1>}, {limitMax=<15 15 15>, limitMin=<-15 -15 -15>}]
    rotationOffsets [<0 0 0>, <0 0 0>]
    trailBone {boneName=dyn_braid_01_f}
   }
   n14 = RigidBody!(ComponentPose=@n13, Alpha, OverrideWorldGravity:Vector=Get gravity_override() -> (gravity_override:Vector), ExternalForce=Get Velocity()) -> (Pose) {
    bClampLinearTranslationLimitToRefPose=true
    bOverrideWorldGravity=true
    overridePhysicsAsset=&/F_MED_Rays_FaceAcc_Physics
    simulationSpace=BaseBoneSpace
    alphaBoolBlend {blendOption=Linear}
    baseBoneRef {boneName=neck_02}
    componentAppliedLinearAccClamp=<6000 6000 6000>
    componentLinearAccScale=<0.1 0.8 0.3>
    componentLinearVelScale=<0.1 1 1>
    overrideWorldGravity {z=-980}
    simSpaceSettings {
     maxAngularAcceleration=100
     maxAngularVelocity=3
     maxLinearAcceleration=400
     maxLinearVelocity=250
     velocityScaleZ=0
     worldAlpha=1
     externalLinearDragV=<0.1 0.1>
    }
   }
   n15 = Component To Local(ComponentPose=@n14)
   n16 = Trail controller - Bone: dyn_braid_01_a(ComponentPose=Local To Component(LocalPose=Use cached pose 'BackpackCache'()), Alpha=1) {
    bLimitRotation=true
    stretchLimit=3000
    rotationLimits [{limitMax=<1 1 1>, limitMin=<-1 -1 -1>}, {limitMax=<15 15 15>, limitMin=<-15 -15 -15>}]
    rotationOffsets [<0 0 0>, <0 0 0>]
    trailBone {boneName=dyn_braid_01_a}
   }
   n17 = Trail controller - Bone: dyn_braid_01_b(ComponentPose=@n16, Alpha=1) {
    bLimitRotation=true
    stretchLimit=3000
    rotationLimits [{limitMax=<1 1 1>, limitMin=<-1 -1 -1>}, {limitMax=<15 15 15>, limitMin=<-15 -15 -15>}]
    rotationOffsets [<0 0 0>, <0 0 0>]
    trailBone {boneName=dyn_braid_01_b}
   }
   n18 = Trail controller - Bone: dyn_braid_01_c(ComponentPose=@n17, Alpha=1) {
    bLimitRotation=true
    stretchLimit=3000
    rotationLimits [{limitMax=<1 1 1>, limitMin=<-1 -1 -1>}, {limitMax=<15 15 15>, limitMin=<-15 -15 -15>}]
    rotationOffsets [<0 0 0>, <0 0 0>]
    trailBone {boneName=dyn_braid_01_c}
   }
   n19 = Trail controller - Bone: dyn_braid_01_d(ComponentPose=@n18, Alpha=1) {
    bLimitRotation=true
    stretchLimit=3000
    rotationLimits [{limitMax=<1 1 1>, limitMin=<-1 -1 -1>}, {limitMax=<15 15 15>, limitMin=<-15 -15 -15>}]
    rotationOffsets [<0 0 0>, <0 0 0>]
    trailBone {boneName=dyn_braid_01_d}
   }
   n20 = Trail controller - Bone: dyn_braid_01_e(ComponentPose=@n19, Alpha=1) {
    bLimitRotation=true
    stretchLimit=3000
    rotationLimits [{limitMax=<1 1 1>, limitMin=<-1 -1 -1>}, {limitMax=<15 15 15>, limitMin=<-15 -15 -15>}]
    rotationOffsets [<0 0 0>, <0 0 0>]
    trailBone {boneName=dyn_braid_01_e}
   }
   n21 = Trail controller - Bone: dyn_braid_01_f(ComponentPose=@n20, Alpha=1) {
    bLimitRotation=true
    stretchLimit=3000
    rotationLimits [{limitMax=<1 1 1>, limitMin=<-1 -1 -1>}, {limitMax=<15 15 15>, limitMin=<-15 -15 -15>}]
    rotationOffsets [<0 0 0>, <0 0 0>]
    trailBone {boneName=dyn_braid_01_f}
   }
   n22 = RigidBody(ComponentPose=@n21, OverrideWorldGravity=Get gravity_override(), ExternalForce=Get Velocity()) {
    bClampLinearTranslationLimitToRefPose=true
    bOverrideWorldGravity=true
    overridePhysicsAsset=&F_MED_Rays_FaceAcc_Physics
    simulationSpace=BaseBoneSpace
    alphaBoolBlend {blendOption=Linear}
    baseBoneRef {boneName=neck_02}
    componentAppliedLinearAccClamp=<6000 6000 6000>
    componentLinearAccScale=<0.1 0.8 0.3>
    componentLinearVelScale=<0.1 1 1>
    overrideWorldGravity {z=-980}
    simSpaceSettings {
     maxAngularAcceleration=100
     maxAngularVelocity=3
     maxLinearAcceleration=400
     maxLinearVelocity=250
     velocityScaleZ=0
     worldAlpha=1
     externalLinearDragV=<0.1 0.1>
    }
   }
   n23 = Component To Local(ComponentPose=@n22)
   n24 = Get bRigidBodyEnabled()
   n25 = Trail controller - Bone: dyn_braid_01_a(ComponentPose=Local To Component(LocalPose=Use cached pose 'BackpackCache'()), Alpha=1) {
    bLimitRotation=true
    stretchLimit=3000
    rotationLimits [{limitMax=<1 1 1>, limitMin=<-1 -1 -1>}, {limitMax=<15 15 15>, limitMin=<-15 -15 -15>}]
    rotationOffsets [<0 0 0>, <0 0 0>]
    trailBone {boneName=dyn_braid_01_a}
   }
   n26 = Trail controller - Bone: dyn_braid_01_b(ComponentPose=@n25, Alpha=1) {
    bLimitRotation=true
    stretchLimit=3000
    rotationLimits [{limitMax=<1 1 1>, limitMin=<-1 -1 -1>}, {limitMax=<15 15 15>, limitMin=<-15 -15 -15>}]
    rotationOffsets [<0 0 0>, <0 0 0>]
    trailBone {boneName=dyn_braid_01_b}
   }
   n27 = Trail controller - Bone: dyn_braid_01_c(ComponentPose=@n26, Alpha=1) {
    bLimitRotation=true
    stretchLimit=3000
    rotationLimits [{limitMax=<1 1 1>, limitMin=<-1 -1 -1>}, {limitMax=<15 15 15>, limitMin=<-15 -15 -15>}]
    rotationOffsets [<0 0 0>, <0 0 0>]
    trailBone {boneName=dyn_braid_01_c}
   }
   n28 = Trail controller - Bone: dyn_braid_01_d(ComponentPose=@n27, Alpha=1) {
    bLimitRotation=true
    stretchLimit=3000
    rotationLimits [{limitMax=<1 1 1>, limitMin=<-1 -1 -1>}, {limitMax=<15 15 15>, limitMin=<-15 -15 -15>}]
    rotationOffsets [<0 0 0>, <0 0 0>]
    trailBone {boneName=dyn_braid_01_d}
   }
   n29 = Trail controller - Bone: dyn_braid_01_e(ComponentPose=@n28, Alpha=1) {
    bLimitRotation=true
    stretchLimit=3000
    rotationLimits [{limitMax=<1 1 1>, limitMin=<-1 -1 -1>}, {limitMax=<15 15 15>, limitMin=<-15 -15 -15>}]
    rotationOffsets [<0 0 0>, <0 0 0>]
    trailBone {boneName=dyn_braid_01_e}
   }
   n30 = Trail controller - Bone: dyn_braid_01_f(ComponentPose=@n29, Alpha=1) {
    bLimitRotation=true
    stretchLimit=3000
    rotationLimits [{limitMax=<1 1 1>, limitMin=<-1 -1 -1>}, {limitMax=<15 15 15>, limitMin=<-15 -15 -15>}]
    rotationOffsets [<0 0 0>, <0 0 0>]
    trailBone {boneName=dyn_braid_01_f}
   }
   n31 = "Transform (Modify) Bone - Bone: dyn_hair_base"(
    ComponentPose=@n30
    Alpha=Get Hair_base_alpha() -> (Hair_base_alpha:float)
    Translation="-0.516,0.000,-2.391"
    Rotation="-7.507,-0.000,-0.000"
    Scale="1,1,1"
   ) {
    rotationMode=BMM_Additive
    rotationSpace=BCS_BoneSpace
    scaleMode=BMM_Additive
    scaleSpace=BCS_BoneSpace
    translationMode=BMM_Additive
    translationSpace=BCS_BoneSpace
    alphaBoolBlend {blendOption=Linear}
    boneToModify {boneName=dyn_hair_base}
    rotation=<rot -7.5069 0 0>
    translation {x=-0.516, z=-2.391}
   }
   n32 = "Transform (Modify) Bone - Bone: dyn_hair_base"(ComponentPose=@n31, Alpha=Get Skydive(), Translation="0,0,0", Rotation="-60,0.000,-0.000", Scale="1,1,1") {
    rotationMode=BMM_Additive
    rotationSpace=BCS_BoneSpace
    translationMode=BMM_Additive
    translationSpace=BCS_BoneSpace
    alphaBoolBlend {blendOption=Linear}
    boneToModify {boneName=dyn_hair_base}
    rotation {pitch=-60}
   }
   n33 = RigidBody(ComponentPose=@n32, OverrideWorldGravity=Get gravity_override(), ExternalForce=Get Velocity()) {
    bClampLinearTranslationLimitToRefPose=true
    bOverrideWorldGravity=true
    overridePhysicsAsset=&/F_MED_Rays_FaceAcc_CapeEquipped_Physics
    simulationSpace=BaseBoneSpace
    alphaBoolBlend {blendOption=Linear}
    baseBoneRef {boneName=neck_02}
    componentAppliedLinearAccClamp=<6000 6000 6000>
    componentLinearAccScale=<0.1 1 0.3>
    componentLinearVelScale=<0.1 1 1>
    overrideWorldGravity {z=-980}
    simSpaceSettings {
     maxAngularAcceleration=100
     maxAngularVelocity=3
     maxLinearAcceleration=400
     maxLinearVelocity=250
     velocityScaleZ=0
     worldAlpha=1
     externalLinearDragV=<0.1 0.1>
    }
   }
   n34 = Component To Local(ComponentPose=@n33)
   n35 = "Constraint - dyn_braid_parent_a (PP)"!(
    ComponentPose=Local To Component(LocalPose=Use cached pose 'BackpackCache'())
    Alpha=1
    ConstraintWeights_0=0.75
    ConstraintWeights_1:float=0.25
   ) -> (Pose) {
    alphaBoolBlend {blendOption=Linear}
    boneToModify {boneName=dyn_braid_parent_a}
    constraintSetup [{targetBone {boneName=head}}, {targetBone {boneName=neck_01}}]
    constraintWeights [1,1]
   }
   n36 = "Constraint - dyn_braid_parent_b (PP)"!(ComponentPose=@n35, Alpha=1, ConstraintWeights_0=0.75, ConstraintWeights_1:float=0.25) -> (Pose) {
    alphaBoolBlend {blendOption=Linear}
    boneToModify {boneName=dyn_braid_parent_b}
    constraintSetup [{targetBone {boneName=head}}, {targetBone {boneName=neck_01}}]
    constraintWeights [1,1]
   }
   n37 = "Constraint - dyn_braid_parent_c (PP)"!(ComponentPose=@n36, Alpha=1, ConstraintWeights_0=0.75, ConstraintWeights_1:float=0.25) -> (Pose) {
    alphaBoolBlend {blendOption=Linear}
    boneToModify {boneName=dyn_braid_parent_c}
    constraintSetup [{targetBone {boneName=head}}, {targetBone {boneName=neck_01}}]
    constraintWeights [1,1]
   }
   n38 = "Constraint - dyn_braid_parent_d (PP)"!(ComponentPose=@n37, Alpha=1, ConstraintWeights_0=0.75, ConstraintWeights_1:float=0.25) -> (Pose) {
    alphaBoolBlend {blendOption=Linear}
    boneToModify {boneName=dyn_braid_parent_d}
    constraintSetup [{targetBone {boneName=head}}, {targetBone {boneName=neck_01}}]
    constraintWeights [1,1]
   }
   n39 = "Constraint - dyn_braid_parent_e (PP)"!(ComponentPose=@n38, Alpha=1, ConstraintWeights_0=0.75, ConstraintWeights_1:float=0.25) -> (Pose) {
    alphaBoolBlend {blendOption=Linear}
    boneToModify {boneName=dyn_braid_parent_e}
    constraintSetup [{targetBone {boneName=head}}, {targetBone {boneName=neck_01}}]
    constraintWeights [1,1]
   }
   n40 = "Constraint - dyn_braid_parent_f (PP)"!(ComponentPose=@n39, Alpha=1, ConstraintWeights_0=0.75, ConstraintWeights_1:float=0.25) -> (Pose) {
    alphaBoolBlend {blendOption=Linear}
    boneToModify {boneName=dyn_braid_parent_f}
    constraintSetup [{targetBone {boneName=head}}, {targetBone {boneName=neck_01}}]
    constraintWeights [1,1]
   }
   n41 = "Constraint - dyn_braid_01_a (PP)"!(ComponentPose=@n40, Alpha=1, ConstraintWeights_0=0.5, ConstraintWeights_1:float=0.5) -> (Pose) {
    alphaBoolBlend {blendOption=Linear}
    boneToModify {boneName=dyn_braid_01_a}
    constraintSetup [{targetBone {boneName=spine_05}}, {targetBone {boneName=neck_01}}]
    constraintWeights [1,1]
   }
   n42 = "Constraint - dyn_braid_01_b (PP)"!(ComponentPose=@n41, Alpha=1, ConstraintWeights_0=0.5, ConstraintWeights_1:float=0.5) -> (Pose) {
    alphaBoolBlend {blendOption=Linear}
    boneToModify {boneName=dyn_braid_01_b}
    constraintSetup [{targetBone {boneName=spine_05}}, {targetBone {boneName=neck_01}}]
    constraintWeights [1,1]
   }
   n43 = "Constraint - dyn_braid_01_c (PP)"!(ComponentPose=@n42, Alpha=1, ConstraintWeights_0=0.5, ConstraintWeights_1:float=0.5) -> (Pose) {
    alphaBoolBlend {blendOption=Linear}
    boneToModify {boneName=dyn_braid_01_c}
    constraintSetup [{targetBone {boneName=spine_05}}, {targetBone {boneName=neck_01}}]
    constraintWeights [1,1]
   }
   n44 = "Constraint - dyn_braid_01_d (PP)"!(ComponentPose=@n43, Alpha=1, ConstraintWeights_0=0.5, ConstraintWeights_1:float=0.5) -> (Pose) {
    alphaBoolBlend {blendOption=Linear}
    boneToModify {boneName=dyn_braid_01_d}
    constraintSetup [{targetBone {boneName=spine_05}}, {targetBone {boneName=neck_01}}]
    constraintWeights [1,1]
   }
   n45 = "Constraint - dyn_braid_01_e (PP)"!(ComponentPose=@n44, Alpha=1, ConstraintWeights_0=0.5, ConstraintWeights_1:float=0.5) -> (Pose) {
    alphaBoolBlend {blendOption=Linear}
    boneToModify {boneName=dyn_braid_01_e}
    constraintSetup [{targetBone {boneName=spine_05}}, {targetBone {boneName=neck_01}}]
    constraintWeights [1,1]
   }
   n46 = "Constraint - dyn_braid_01_f (PP)"!(ComponentPose=@n45, Alpha=1, ConstraintWeights_0=0.5, ConstraintWeights_1:float=0.5) -> (Pose) {
    alphaBoolBlend {blendOption=Linear}
    boneToModify {boneName=dyn_braid_01_f}
    constraintSetup [{targetBone {boneName=spine_05}}, {targetBone {boneName=neck_01}}]
    constraintWeights [1,1]
   }
   n47 = "Constraint - dyn_braid_02_a (P)"(ComponentPose=@n46, Alpha=1, ConstraintWeights_0=1) {
    alphaBoolBlend {blendOption=Linear}
    boneToModify {boneName=dyn_braid_02_a}
    constraintSetup [{targetBone {boneName=spine_05}}]
    constraintWeights [1]
   }
   n48 = "Constraint - dyn_braid_02_b (P)"(ComponentPose=@n47, Alpha=1, ConstraintWeights_0=1) {
    alphaBoolBlend {blendOption=Linear}
    boneToModify {boneName=dyn_braid_02_b}
    constraintSetup [{targetBone {boneName=spine_05}}]
    constraintWeights [1]
   }
   n49 = "Constraint - dyn_braid_02_c (P)"(ComponentPose=@n48, Alpha=1, ConstraintWeights_0=1) {
    alphaBoolBlend {blendOption=Linear}
    boneToModify {boneName=dyn_braid_02_c}
    constraintSetup [{targetBone {boneName=spine_05}}]
    constraintWeights [1]
   }
   n50 = "Constraint - dyn_braid_02_d (P)"(ComponentPose=@n49, Alpha=1, ConstraintWeights_0=1) {
    alphaBoolBlend {blendOption=Linear}
    boneToModify {boneName=dyn_braid_02_d}
    constraintSetup [{targetBone {boneName=spine_05}}]
    constraintWeights [1]
   }
   n51 = "Constraint - dyn_braid_02_e (P)"(ComponentPose=@n50, Alpha=1, ConstraintWeights_0=1) {
    alphaBoolBlend {blendOption=Linear}
    boneToModify {boneName=dyn_braid_02_e}
    constraintSetup [{targetBone {boneName=spine_05}}]
    constraintWeights [1]
   }
   n52 = "Constraint - dyn_braid_02_f (P)"(ComponentPose=@n51, Alpha=1, ConstraintWeights_0=1) {
    alphaBoolBlend {blendOption=Linear}
    boneToModify {boneName=dyn_braid_02_f}
    constraintSetup [{targetBone {boneName=spine_05}}]
    constraintWeights [1]
   }
   n53 = "Transform (Modify) Bone - Bone: dyn_braid_parent_a"(ComponentPose=@n52, Alpha=1, Translation="0,0,0", Rotation="3,0,0", Scale="1,1,1") {
    rotationMode=BMM_Additive
    rotationSpace=BCS_BoneSpace
    scaleMode=BMM_Additive
    scaleSpace=BCS_BoneSpace
    translationMode=BMM_Additive
    translationSpace=BCS_BoneSpace
    alphaBoolBlend {blendOption=Linear}
    boneToModify {boneName=dyn_braid_parent_a}
    rotation {pitch=3}
   }
   n54 = "Transform (Modify) Bone - Bone: dyn_braid_parent_b"(ComponentPose=@n53, Alpha=1, Translation="0,0,0", Rotation="7,0,0", Scale="1,1,1") {
    rotationMode=BMM_Additive
    rotationSpace=BCS_BoneSpace
    scaleMode=BMM_Additive
    scaleSpace=BCS_BoneSpace
    translationMode=BMM_Additive
    translationSpace=BCS_BoneSpace
    alphaBoolBlend {blendOption=Linear}
    boneToModify {boneName=dyn_braid_parent_b}
    rotation {pitch=7}
   }
   n55 = "Transform (Modify) Bone - Bone: dyn_braid_parent_c"(ComponentPose=@n54, Alpha=1, Translation="0,0,0", Rotation="7,0,0", Scale="1,1,1") {
    rotationMode=BMM_Additive
    rotationSpace=BCS_BoneSpace
    scaleMode=BMM_Additive
    scaleSpace=BCS_BoneSpace
    translationMode=BMM_Additive
    translationSpace=BCS_BoneSpace
    alphaBoolBlend {blendOption=Linear}
    boneToModify {boneName=dyn_braid_parent_c}
    rotation {pitch=7}
   }
   n56 = "Transform (Modify) Bone - Bone: dyn_braid_parent_d"(ComponentPose=@n55, Alpha=1, Translation="0,0,0", Rotation="2,0,0", Scale="1,1,1") {
    rotationMode=BMM_Additive
    rotationSpace=BCS_BoneSpace
    scaleMode=BMM_Additive
    scaleSpace=BCS_BoneSpace
    translationMode=BMM_Additive
    translationSpace=BCS_BoneSpace
    alphaBoolBlend {blendOption=Linear}
    boneToModify {boneName=dyn_braid_parent_d}
    rotation {pitch=2}
   }
   n57 = "Transform (Modify) Bone - Bone: dyn_braid_parent_e"(ComponentPose=@n56, Alpha=1, Translation="0,0,0", Rotation="5,0,0", Scale="1,1,1") {
    rotationMode=BMM_Additive
    rotationSpace=BCS_BoneSpace
    scaleMode=BMM_Additive
    scaleSpace=BCS_BoneSpace
    translationMode=BMM_Additive
    translationSpace=BCS_BoneSpace
    alphaBoolBlend {blendOption=Linear}
    boneToModify {boneName=dyn_braid_parent_e}
    rotation {pitch=5}
   }
   n58 = "Transform (Modify) Bone - Bone: dyn_braid_parent_f"(ComponentPose=@n57, Alpha=1, Translation="0,0,0", Rotation="3,0,0", Scale="1,1,1") {
    rotationMode=BMM_Additive
    rotationSpace=BCS_BoneSpace
    scaleMode=BMM_Additive
    scaleSpace=BCS_BoneSpace
    translationMode=BMM_Additive
    translationSpace=BCS_BoneSpace
    alphaBoolBlend {blendOption=Linear}
    boneToModify {boneName=dyn_braid_parent_f}
    rotation {pitch=3}
   }
   n59 = "Transform (Modify) Bone - Bone: dyn_braid_01_a"(ComponentPose=@n58, Alpha=1, Translation="0,0,0", Rotation="2,0,0", Scale="1,1,1") {
    rotationMode=BMM_Additive
    rotationSpace=BCS_BoneSpace
    scaleMode=BMM_Additive
    scaleSpace=BCS_BoneSpace
    translationMode=BMM_Additive
    translationSpace=BCS_BoneSpace
    alphaBoolBlend {blendOption=Linear}
    boneToModify {boneName=dyn_braid_01_a}
    rotation {pitch=2}
   }
   n60 = "Transform (Modify) Bone - Bone: dyn_braid_01_b"(ComponentPose=@n59, Alpha=1, Translation="0,0,0", Rotation="3,0,0", Scale="1,1,1") {
    rotationMode=BMM_Additive
    rotationSpace=BCS_BoneSpace
    scaleMode=BMM_Additive
    scaleSpace=BCS_BoneSpace
    translationMode=BMM_Additive
    translationSpace=BCS_BoneSpace
    alphaBoolBlend {blendOption=Linear}
    boneToModify {boneName=dyn_braid_01_b}
    rotation {pitch=3}
   }
   n61 = "Transform (Modify) Bone - Bone: dyn_braid_01_c"(ComponentPose=@n60, Alpha=1, Translation="0,0,0", Rotation="3,0,0", Scale="1,1,1") {
    rotationMode=BMM_Additive
    rotationSpace=BCS_BoneSpace
    scaleMode=BMM_Additive
    scaleSpace=BCS_BoneSpace
    translationMode=BMM_Additive
    translationSpace=BCS_BoneSpace
    alphaBoolBlend {blendOption=Linear}
    boneToModify {boneName=dyn_braid_01_c}
    rotation {pitch=3}
   }
   n62 = "Transform (Modify) Bone - Bone: dyn_braid_01_d"(ComponentPose=@n61, Alpha=1, Translation="0,0,0", Rotation="3,0,0", Scale="1,1,1") {
    rotationMode=BMM_Additive
    rotationSpace=BCS_BoneSpace
    scaleMode=BMM_Additive
    scaleSpace=BCS_BoneSpace
    translationMode=BMM_Additive
    translationSpace=BCS_BoneSpace
    alphaBoolBlend {blendOption=Linear}
    boneToModify {boneName=dyn_braid_01_d}
    rotation {pitch=3}
   }
   n63 = "Transform (Modify) Bone - Bone: dyn_braid_01_e"(ComponentPose=@n62, Alpha=1, Translation="0,0,0", Rotation="3,0,0", Scale="1,1,1") {
    rotationMode=BMM_Additive
    rotationSpace=BCS_BoneSpace
    scaleMode=BMM_Additive
    scaleSpace=BCS_BoneSpace
    translationMode=BMM_Additive
    translationSpace=BCS_BoneSpace
    alphaBoolBlend {blendOption=Linear}
    boneToModify {boneName=dyn_braid_01_e}
    rotation {pitch=3}
   }
   n64 = "Transform (Modify) Bone - Bone: dyn_braid_01_f"(ComponentPose=@n63, Alpha=1, Translation="0,0,0", Rotation="3,0,0", Scale="1,1,1") {
    rotationMode=BMM_Additive
    rotationSpace=BCS_BoneSpace
    scaleMode=BMM_Additive
    scaleSpace=BCS_BoneSpace
    translationMode=BMM_Additive
    translationSpace=BCS_BoneSpace
    alphaBoolBlend {blendOption=Linear}
    boneToModify {boneName=dyn_braid_01_f}
    rotation {pitch=3}
   }
   n65 = "Transform (Modify) Bone - Bone: dyn_braid_02_a"(ComponentPose=@n64, Alpha=1, Translation="0,0,2", Rotation="10,0,0", Scale="1,1,1") {
    rotationMode=BMM_Additive
    rotationSpace=BCS_BoneSpace
    scaleMode=BMM_Additive
    scaleSpace=BCS_BoneSpace
    translationMode=BMM_Additive
    translationSpace=BCS_BoneSpace
    alphaBoolBlend {blendOption=Linear}
    boneToModify {boneName=dyn_braid_02_a}
    rotation {pitch=10}
    translation {z=2}
   }
   n66 = "Transform (Modify) Bone - Bone: dyn_braid_02_b"(ComponentPose=@n65, Alpha=1, Translation="0,0,2", Rotation="10,0,0", Scale="1,1,1") {
    rotationMode=BMM_Additive
    rotationSpace=BCS_BoneSpace
    scaleMode=BMM_Additive
    scaleSpace=BCS_BoneSpace
    translationMode=BMM_Additive
    translationSpace=BCS_BoneSpace
    alphaBoolBlend {blendOption=Linear}
    boneToModify {boneName=dyn_braid_02_b}
    rotation {pitch=10}
    translation {z=2}
   }
   n67 = "Transform (Modify) Bone - Bone: dyn_braid_02_c"(ComponentPose=@n66, Alpha=1, Translation="0,0,2", Rotation="10,0,0", Scale="1,1,1") {
    rotationMode=BMM_Additive
    rotationSpace=BCS_BoneSpace
    scaleMode=BMM_Additive
    scaleSpace=BCS_BoneSpace
    translationMode=BMM_Additive
    translationSpace=BCS_BoneSpace
    alphaBoolBlend {blendOption=Linear}
    boneToModify {boneName=dyn_braid_02_c}
    rotation {pitch=10}
    translation {z=2}
   }
   n68 = "Transform (Modify) Bone - Bone: dyn_braid_02_d"(ComponentPose=@n67, Alpha=1, Translation="0,0,1", Rotation="10,0,0", Scale="1,1,1") {
    rotationMode=BMM_Additive
    rotationSpace=BCS_BoneSpace
    scaleMode=BMM_Additive
    scaleSpace=BCS_BoneSpace
    translationMode=BMM_Additive
    translationSpace=BCS_BoneSpace
    alphaBoolBlend {blendOption=Linear}
    boneToModify {boneName=dyn_braid_02_d}
    rotation {pitch=10}
    translation {z=1}
   }
   n69 = "Transform (Modify) Bone - Bone: dyn_braid_02_e"(ComponentPose=@n68, Alpha=1, Translation="0,0,2", Rotation="10,0,0", Scale="1,1,1") {
    rotationMode=BMM_Additive
    rotationSpace=BCS_BoneSpace
    scaleMode=BMM_Additive
    scaleSpace=BCS_BoneSpace
    translationMode=BMM_Additive
    translationSpace=BCS_BoneSpace
    alphaBoolBlend {blendOption=Linear}
    boneToModify {boneName=dyn_braid_02_e}
    rotation {pitch=10}
    translation {z=2}
   }
   n70 = "Transform (Modify) Bone - Bone: dyn_braid_02_f"(ComponentPose=@n69, Alpha=1, Translation="0,0,2", Rotation="10,0,0", Scale="1,1,1") {
    rotationMode=BMM_Additive
    rotationSpace=BCS_BoneSpace
    scaleMode=BMM_Additive
    scaleSpace=BCS_BoneSpace
    translationMode=BMM_Additive
    translationSpace=BCS_BoneSpace
    alphaBoolBlend {blendOption=Linear}
    boneToModify {boneName=dyn_braid_02_f}
    rotation {pitch=10}
    translation {z=2}
   }
   n71 = Component To Local(ComponentPose=@n70)
   n72 = Get isWearingCape() -> (isWearingCape:bool)
   Output Pose(Result=Blend Poses by bool(
     BlendPose_0=Blend Poses by bool(BlendPose_0=@n34, BlendPose_1=@n71, bActiveValue=@n72) {blendType=Linear}
     BlendPose_1=Blend Poses by bool(BlendPose_0=@n23, BlendPose_1=@n15, bActiveValue=@n24) {blendType=Linear}
     bActiveValue=Get isWearingPack() -> (isWearingPack:bool)
    ) {blendType=Linear}) {layerGroup=None}
   n73 = Copy Pose From Mesh(SourceMeshComponent=Get BodyMesh() -> (BodyMesh:Skeletal Mesh Component Object Reference)) {bUseAttachedParent=true}
   n74 = Copy Pose From Mesh() {bUseAttachedParent=true}
   n75 = Anim Dynamics - Bone: dyn_hair_tie_parent_r(
    ComponentPose=Local To Component(LocalPose=Layered blend per bone(BasePose=@n74, BlendPoses_0=@n73) {layerSetup [{branchFilters [{boneName=upperarm_l}, {boneName=upperarm_r}]}]})
    Alpha=1
    ExternalForce="0,0,0"
   ) {
    angularDampingOverride=0.9999
    angularSpringConstant=100
    bAngularSpring=true
    bOverrideAngularDamping=true
    bOverrideLinearDamping=true
    bUsePlanarLimit=false
    bUseSphericalLimits=true
    linearDampingOverride=0.9999
    alphaBoolBlend {blendOption=Linear}
    boundBone {boneName=dyn_hair_tie_parent_r}
    componentAppliedLinearAccClamp=<6000 6000 6000>
    componentLinearAccScale=<0.1 0.1 0.1>
    componentLinearVelScale=<0.1 0.1 0.1>
    physicsBodyDefinitions [{
      collisionType=CustomSphere
      sphereCollisionRadius=3
      boundBone {boneName=dyn_hair_tie_parent_r}
      boxExtents=<12 4 4>
      constraintSetup {angularLimitsMax=<1 1 12>, angularLimitsMin=<0 0 -20>, angularTarget=<1 0 0>}
      localJointOffset=<0.1 0 0>
     }]
    sphericalLimits [{limitRadius=5, drivingBone {boneName=head_offset}, sphereLocalOffset=<0 0 0>}]
   }
   n76 = Anim Dynamics - Bone: dyn_hair_tie_parent_l(ComponentPose=@n75, Alpha=1, ExternalForce="0,0,0") {
    angularDampingOverride=0.9999
    angularSpringConstant=100
    bAngularSpring=true
    bOverrideAngularDamping=true
    bOverrideLinearDamping=true
    bUseSphericalLimits=true
    linearDampingOverride=0.9999
    alphaBoolBlend {blendOption=Linear}
    boundBone {boneName=dyn_hair_tie_parent_l}
    componentAppliedLinearAccClamp=<6000 6000 6000>
    componentLinearAccScale=<0.1 0.1 0.1>
    componentLinearVelScale=<0.1 0.1 0.1>
    physicsBodyDefinitions [{
      collisionType=CustomSphere
      sphereCollisionRadius=2
      boundBone {boneName=dyn_hair_tie_parent_l}
      boxExtents=<12 4 4>
      constraintSetup {angularLimitsMax=<0 80 10>, angularLimitsMin=<0 -1 -30>, angularTarget=<0 0 0>}
      localJointOffset=<1 0 0>
     }]
    sphericalLimits [{limitRadius=5, drivingBone {boneName=head_offset}, sphereLocalOffset=<0 0 0>}]
   }
   n77 = Look At - Bone: dyn_hair_base(ComponentPose=@n76, Alpha=0) {
    alphaBoolBlend {blendOption=Linear}
    boneToModify {boneName=dyn_hair_base}
    lookAtLocation=<9.5801 -18.2168 -0.2114>
    lookAtTarget {boneReference {boneName=spine_03}}
    lookAt_Axis {axis=<1 0>}
    lookUp_Axis {axis {x=0, z=-1}}
   }
   n78 = "Transform (Modify) Bone - Bone: dyn_hair_base"(ComponentPose=@n77, Alpha=Get Hair_base_alpha(), Translation="0,0,0", Rotation="-34.183,-0.159,-0.609", Scale="1,1,1") {
    rotationMode=BMM_Additive
    rotationSpace=BCS_BoneSpace
    scaleMode=BMM_Additive
    scaleSpace=BCS_BoneSpace
    translationMode=BMM_Additive
    translationSpace=BCS_BoneSpace
    alphaBoolBlend {blendOption=Linear}
    boneToModify {boneName=dyn_hair_base}
    rotation=<rot -34.183 -0.159 -0.609>
   }
   Save cached pose 'BackpackCache'(Pose=Component To Local(ComponentPose=@n78)) {cacheName=BackpackCache}
  ]}
 EventGraph {nodes [
   Event BlueprintUpdateAnimation()
   > n1 = Sequence!(execute) -> (then_0, then_1, then_2:Exec, then_3:Exec)
   > Set Hair_base_alpha(Hair_base_alpha:float=Get Distance Between Two Sockets!(
     Component:Skeletal Mesh Component Object Reference=GetOwningComponent()
     SocketOrBoneNameA:Name=dyn_hair_base
     SocketSpaceA:ERelativeTransformSpace Enum=RTS_World
     SocketOrBoneNameB:Name=clavicle_r
     SocketSpaceB:ERelativeTransformSpace Enum=RTS_World
     bRemapRange:bool=true
     InRangeMin:float=28
     InRangeMax:float=17
     OutRangeMin:float=0
     OutRangeMax:float=1
    ) -> (ReturnValue:float) {is=title K2Node_CallFunction}) -> (Output_Get:float)
   n2 = Get bIsSkydiveDiveMode()
   n3 = Get bIsSkydiveFloating()
   > Set Velocity(Velocity:Vector=SelectVector(A="0, 0,56000", B="0, 0, 0", bPickA=OR Boolean(A=@n2, B=@n3))) -> (Output_Get:Vector)
   n4 = OR Boolean!(A=Get bIsSprinting(), B=Get bIsCrouchSprinting(), C:bool=AND Boolean(A=Get bIsCrouching(), B=Get bIsMoving2D()), D:bool=Get bIsMoving2D()) -> (ReturnValue)
   n5 = FInterpTo(
    Current=Get CapeAlpha() -> (CapeAlpha:double)
    Target="To Float (Boolean)"(InBool=@n4)
    DeltaTime=GetWorldDeltaSeconds!() -> (ReturnValue:double)
    InterpSpeed=25
   )
   > Set CapeAlpha(CapeAlpha:double=@n5) -> (Output_Get:double)
   n6 = float * float(A=5000, B=@n5) {is=title K2Node_PromotableOperator}
   n7 = TryGetPawnOwner()
   n8 = vector * vector(A=GetActorForwardVector!(self:Actor Object Reference=@n7) -> (ReturnValue:Vector), B=-1) -> (ReturnValue=~)
   n9 = GetActorUpVector!(self:Actor Object Reference=@n7) -> (ReturnValue:Vector)
   n10 = SelectFloat(A=6000, B=0, bPickA=AND Boolean(A=Get isWearingCape(), B=@n4))
   > Set Velocity(Velocity=vector + vector(A=vector * vector(A=@n8, B=@n6) -> (ReturnValue=~), B=vector * vector(A=@n9, B=@n10) -> (ReturnValue=~)))
   n11 = Gravity Override State Function DEPRECATED(execute=@n1.then_2, CharacterPartAnimInstance=Self-Reference(), GameTimeInSeconds=GetGameTimeInSeconds())
   > Set gravity_override(gravity_override=@n11)
   n12 = Get bIsSkydiveFloating()
   n13 = Get bIsSkydiveDiveMode()
   Set Skydive(execute=@n1.then_1, Skydive:double=SelectFloat(A=1, B=0, bPickA=OR Boolean!(A=@n13, B=@n12) -> (ReturnValue))) -> (Output_Get:double)
   n14 = Event OnCharacterCustomizationLoaded()
   n15 = CheckForBackbling()
   > Set isWearingPack(isWearingPack:bool=@n15) -> (Output_Get:bool)
   > Set isWearingCape(isWearingCape=AND Boolean(A=@n15, B=@n15.bIsBackblingLong))
   > Set BodyMesh(BodyMesh="Get (a copy)"(Array:Array of Skeletal Mesh Component Budgeted Object References=@n14.CustomMeshParts, Dimension 1=1) -> (Output:Skeletal Mesh Component Budgeted Object Reference))
  ]}
]
variables [
 BodyMesh {type=SkeletalMeshComponent}
 CapeAlpha {type=number}
 gravity_override {type=Vector}
 Hair_base_alpha {type=number}
 isWearingCape {type=boolean}
 isWearingPack {type=boolean}
 Skydive {type=number}
 Velocity {type=Vector}
]
